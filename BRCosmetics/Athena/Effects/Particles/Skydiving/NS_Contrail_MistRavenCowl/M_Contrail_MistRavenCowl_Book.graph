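# Two-sided lit Niagara mesh-particle material for the books and loose pages in the MistRavenCowl skydiving contrail.
# A static "Book" switch selects a textured book or a procedural page with two-tone noisy paper, masked text tinted toward particle color, and dimmed emissive.
# World position offset animates both: books blend sine flapping with pivot-based cover rotation via the BookAnim dynamic parameter, while pages sway about their pivot, varied by PageVariation and UV-masked.
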Material /BRCosmetics/Athena/Effects/Particles/Skydiving/NS_Contrail_MistRavenCowl/M_Contrail_MistRavenCowl_Book
graph {
 properties {
  bAutomaticallySetUsageInEditor=false
  bUsedWithNiagaraMeshParticles=true
  previewMesh=&/SM_Contrail_MistRavenCowl_Book
  twoSided=true
  usedShadingModels=Default Lit
 }
 nodes [
  n1 = Constant()
  n2 = Multiply(
   A=ScalarParameter() {defaultValue=5, parameterName=WPO Amount}
   B=LinearInterpolate(Alpha=MaterialFunctionCall() {materialFunction=&DebugTimeSine}.Result) {constA=1.3, constB=0.5}
  )
  n3 = AppendVector(
   A=MaterialFunctionCall(Color=LocalPosition().XYZ) {materialFunction=&SplitComponents}.G
   B=Constant()
  )
  n4 = TextureSample(UVs=Add(A=Multiply(A=@n3) {constB=0.015}) {constB=0.5}) {samplerType=SAMPLERTYPE_Masks, texture=/BRCosmetics/Accessories/FORT_Backpacks/Backpack_F_MED_CatSpace/Textures/T_Gradient_MASK}
  n5 = TextureSample(UVs=Add(A=Multiply(A=@n3) {constB=0.001}) {constB=0.5}) {
   samplerSource=SSM_Clamp_WorldGroupSettings
   samplerType=SAMPLERTYPE_Masks
   texture=&T_Gradient_MASK
  }
  n6 = Constant() {r=1}
  n7 = Saturate(None=Add(A=MaterialFunctionCall(In=@n5.R, Contrast=@n6) {materialFunction=&CheapContrast}.Result) {constA=0.0001, constB=0})
  n8 = Constant() {r=5}
  n9 = Multiply(
   A=Power(Base=OneMinus(None=@n4.R))
   B=MaterialFunctionCall(Speed=Multiply(A=@n7, B=@n8)) {materialFunction=&DebugTimeSine}.Result
  )
  n10 = Transform(None=MaterialFunctionCall(R=@n1, G=@n1, B=LinearInterpolate(B=@n2, Alpha=@n9) {constB=10}) {materialFunction=&AppendMany}.RGB) {transformSourceType=TRANSFORMSOURCE_Local}
  n11 = WorldPosition()
  n12 = MaterialFunctionCall() {materialFunction=&ObjectPivotPoint}
  n13 = ComponentMask(None=Transform(None=Subtract(A=@n11.XYZ, B=@n12.Object Pivot Location)) {transformSourceType=TRANSFORMSOURCE_World, transformType=TRANSFORM_Instance}) {g=true}
  n14 = Sign(None=@n13)
  n15 = Saturate(None=OneMinus(None=DynamicParameter() {defaultValue=<rgba 1 0 0 1>, paramNames [PageVariation, PageColorLerp, BookAnim, Param4]}.BookAnim))
  n16 = Multiply(A=@n14, B=Multiply(A=NamedRerouteDeclaration(None=@n15) {name=Book Anim, nodeColor=<rgba 1 0.5859 0 1>}) {constB=0.25}) {constA=1, constB=0.25}
  n17 = InverseLinearInterpolate(Value=Abs(None=@n13)) {bClampResult=true, constB=5}
  n18 = Transform(None=MaterialFunctionCall() {materialFunction=&ObjectPivotPoint}.Mesh Particle Pivot Location) {transformSourceType=TRANSFORMSOURCE_World, transformType=TRANSFORM_Instance}
  n19 = Constant()
  n20 = MaterialFunctionCall(R=@n19, G=LinearInterpolate(B=Multiply(A=@n14) {constB=3}, Alpha=@n17) {constB=10}, B=@n19) {materialFunction=&AppendMany}
  n21 = WorldPosition()
  n22 = Transform(None=MaterialFunctionCall!(
    Rotation Amount=LinearInterpolate(B=@n16, Alpha=@n17) {constB=0.5}
    PivotPoint=Add(A=@n18, B=@n20.RGB)
    WorldPosition=Transform(None=@n21.XYZ) {transformSourceType=TRANSFORMSOURCE_World, transformType=TRANSFORM_Instance}
   ) {materialFunction=/Engine/Functions/Engine_MaterialFunctions02/WorldPositionOffset/RotateAboutWorldAxis_cheap}.X-Axis) {transformSourceType=TRANSFORMSOURCE_Instance}
  n23 = InverseLinearInterpolate(Value=NamedRerouteUsage()) {bClampResult=true, constB=0.5}
  n24 = Constant3Vector() {constant=<rgba 0 1 0 0>}
  n25 = Multiply(
   A=Constant() {r=3}
   B=Multiply(
    A=VertexColor()
    B=DynamicParameter() {defaultValue=<rgba 1 0 0 1>, paramNames [PageVariation, PageColorLerp, BookAnim, Param4]}.PageVariation
   ) {constB=1.5}
  )
  n26 = RotateAboutAxis(
   NormalizedRotationAxis=Normalize(VectorInput=Transform(None=@n24) {transformSourceType=TRANSFORMSOURCE_Local})
   RotationAngle=LinearInterpolate(Alpha=MaterialFunctionCall(Speed=@n25) {materialFunction=&DebugTimeSine}.Result) {constA=0.9, constB=1.06}
   PivotPoint=MaterialFunctionCall() {materialFunction=&ObjectPivotPoint}.Object Pivot Location
   Position=WorldPosition().XYZ
  )
  n27 = TextureCoordinate()
  n28 = Saturate(None=OneMinus(None=ComponentMask(None=@n27) {r=true}))
  Multiply(B=StaticSwitchParameter(True=LinearInterpolate(A=@n10, B=@n22, Alpha=@n23), False=Multiply(A=@n26, B=@n28)) {defaultValue=true, parameterName=Book}) {constA=1}
  ParticleSubUVProperties()
  Saturate(None=Add(A=VertexColor().R) {constB=0.5})
  ScalarParameter()
  n29 = TextureSampleParameter2D() {parameterName=Book Texture, texture=/Game/Environments/Helios/Props/Furniture/Furniture_RusticBooks_A/Textures/T_Furniture_RusticBooks_A_A_D}
  n30 = OneMinus(None=MaterialFunctionCall() {materialFunction=&RadialGradientExponential}.RadialGradientExponential)
  n31 = TextureSample(UVs=TextureCoordinate()) {texture=/Game/Global/Textures/Default/Noise/T_godray_Noise}
  n32 = TextureSample() {
   samplerType=SAMPLERTYPE_LinearGrayscale
   texture=&/T_Pickaxe_Contrail_MistRavenCowl_BookPages
  }
  n33 = LinearInterpolate(
   A=LinearInterpolate(
    A=VectorParameter() {parameterName=PaperColor_1, defaultValue=<rgba 0.3419 0.3663 0.3467 1>}.RGB
    B=VectorParameter() {parameterName=PaperColor_2, defaultValue=<rgba 0.125 0.1078 0.1009 1>}.RGB
    Alpha=Multiply(A=@n30, B=@n31.R)
   )
   B=LinearInterpolate(
    A=VectorParameter() {parameterName=TextColor, defaultValue=<rgba 0 0 0 1>}.RGB
    B=ParticleColor().RGB
    Alpha=DynamicParameter() {defaultValue=<rgba 1 0 0 1>, paramNames [PageVariation, PageColorLerp, BookAnim, Param4]}.PageColorLerp
   )
   Alpha=@n32.R
  )
  StaticSwitchParameter(True=@n29.RGB, False=@n33) {defaultValue=true, parameterName=Book}
  StaticSwitchParameter(True=Constant(), False=Multiply(A=@n33) {constB=0.5}) {defaultValue=true, parameterName=Book}
 ]
}
