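# Masked, two-sided mesh-particle material for loose book pages in the MistRavenCowl skydiving contrail.
# Vertex-offset math bends and flaps pages using pivot rotation, vertex color, UV falloff and time sines, gated by a Flapping dynamic parameter.
# Particle parameters select the page sub-image and tint its text over noise-mottled paper colors.
# A noise-based smoothstep erosion, driven by particle Erosion, adds bright fire-curve burn coloring as pages dissolve.
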Material /BRCosmetics/Athena/Effects/Particles/Skydiving/NS_Contrail_MistRavenCowl/M_Contrail_MistRavenCowl_LoosePaper
graph {
 properties {
  bAutomaticallySetUsageInEditor=false
  bHasPixelAnimation=true
  bUsedWithNiagaraMeshParticles=true
  blendMode=BLEND_Masked
  twoSided=true
  usedShadingModels=Default Lit
 }
 nodes [
  n1 = Time()
  ConstantBiasScale(None=Sine(None=Multiply(A=@n1) {constB=0.2}))
  n2 = WorldPosition()
  CrossProduct(A=DDX(Value=@n2.XYZ), B=DDY(Value=@n2.XYZ))
  n3 = Normalize(VectorInput=Transform(None=Constant3Vector() {constant=<rgba 0 1 0 0>}) {transformSourceType=TRANSFORMSOURCE_Local})
  n4 = MaterialFunctionCall(Speed=ScalarParameter() {defaultValue=3, parameterName=WPO_DebugTimeSine}) {materialFunction=&DebugTimeSine}
  n5 = LinearInterpolate(
   A=LinearInterpolate(Alpha=@n4.Result) {constA=0.15, constB=-0.05}
   B=LinearInterpolate(Alpha=@n4.Result) {constA=0.85, constB=1.05}
   Alpha=VertexColor().R
  )
  n6 = TextureCoordinate()
  n7 = Saturate(None=OneMinus(None=ComponentMask(None=@n6) {r=true}))
  n8 = Multiply(
   A=RotateAboutAxis(
    NormalizedRotationAxis=@n3
    RotationAngle=@n5
    PivotPoint=MaterialFunctionCall() {materialFunction=&ObjectPivotPoint}.Object Pivot Location
    Position=WorldPosition().XYZ
   )
   B=@n7
  )
  n9 = Constant()
  n10 = Power(Base=@n7) {constExponent=0.8}
  n11 = MaterialFunctionCall(R=@n9, G=@n9, B=LinearInterpolate(Alpha=Saturate(None=@n10)) {constA=-4, constB=0}) {materialFunction=&AppendMany}
  n12 = Constant()
  n13 = Time()
  n14 = DynamicParameter() {defaultValue=<rgba 0 0 0 0>, paramNames [GlowTransition, Flapping, RandomFlapping, PageColorLerp]}
  n15 = Sine(None=Multiply(A=Add(A=@n13, B=@n14.RandomFlapping) {constB=0.5}) {constB=2.5})
  n16 = LinearInterpolate(Alpha=Divide(A=Add(A=@n15) {constB=0})) {constA=-5, constB=5}
  LinearInterpolate(
   A=Constant()
   B=Add(
    A=Add(A=@n8, B=@n11.RGB)
    B=MaterialFunctionCall(R=@n12, G=@n12, B=@n16) {materialFunction=&AppendMany}.RGB
   )
   Alpha=DynamicParameter() {defaultValue=<rgba 0 0 0 0>, paramNames [GlowTransition, Flapping, RandomFlapping, PageColorLerp]}.Flapping
  )
  n17 = TextureSample(UVs=Multiply(A=TextureCoordinate()) {constB=0.5}) {texture=/Game/Global/Textures/Default/Noise/T_godray_Noise}
  n18 = NamedRerouteUsage()
  n19 = MaterialFunctionCall!(Texture=@n17.R, Erosion=Saturate(None=Subtract(A=@n18) {constB=0.3}), Softness=Constant() {r=0.05}) {materialFunction=/Game/Effects/Niagara/MasterMaterials/Functions/MF_Erosion_Smoothstep_01}
  n20 = OneMinus(None=@n19.Result)
  n21 = MaterialFunctionCall(
   Hue Shift Percentage=Constant() {r=-0.01}
   Texture=CurveAtlasRowParameter(CurveTime=LinearInterpolate(Alpha=@n20) {constB=0.5}) {atlas=/Game/Athena/LinearColorCurves/VFX_CurveAtlas_Fire, curve=/Game/Athena/LinearColorCurves/Curve_Fire03}
  ) {materialFunction=&HueShift}
  n22 = MaterialFunctionCall() {materialFunction=&RadialGradientExponential}
  n23 = LinearInterpolate(
   A=VectorParameter() {parameterName=PaperColor_1, defaultValue=<rgba 0.0535 0.0573 0.0542 1>}.RGB
   B=VectorParameter() {parameterName=PaperColor_2, defaultValue=<rgba 0.125 0.1078 0.1009 1>}.RGB
   Alpha=Multiply(A=OneMinus(None=@n22.RadialGradientExponential), B=@n17.R)
  )
  n24 = DynamicParameter() {defaultValue=<rgba 0 0 0 0>, paramNames [GlowTransition, Flapping, RandomFlapping, PageColorLerp]}
  n25 = DynamicParameter() {parameterIndex=1, defaultValue=<rgba 0 0 1 1>, paramNames [SubUV_Index,Erosion,Param3,Param4]}
  n26 = TextureSample(UVs=MaterialFunctionCall!(UVs, SubImages=Constant2Vector() {g=1, r=3}, Frame=@n25.SubUV_Index) {materialFunction=/BRCosmetics/Animation/Game/MainPlayer/Emotes/TroutWrist/FX/Materials/SubUV_Function}.UVs) {
   samplerType=SAMPLERTYPE_LinearGrayscale
   texture=&/T_Pickaxe_Contrail_MistRavenCowl_BookPages
  }
  n27 = LinearInterpolate(
   A=@n23
   B=LinearInterpolate(A=VectorParameter() {parameterName=TextColor, defaultValue=<rgba 0 0 0 1>}.RGB, B=ParticleColor().RGB, Alpha=@n24.PageColorLerp)
   Alpha=@n26.R
  )
  n28 = Constant3Vector()
  n29 = InverseLinearInterpolate(Value=@n18) {bClampResult=true, constB=0.3}
  MaterialFunctionCall(Emissive=LinearInterpolate(A=Multiply(A=@n21.Result) {constB=20}, B=LinearInterpolate(A=@n27, B=@n28, Alpha=@n29), Alpha=@n19.Result)) {materialFunction=&MF_FX_EmissiveAdjustments}
  MaterialFunctionCall(Rotation Axis=@n3, Rotation Angle=@n5, Normals Pivot=Constant3Vector()) {materialFunction=&FixRotateAboutAxisNormals}
  Multiply(A=LinearInterpolate(A=@n23, B=VectorParameter() {parameterName=Glow Color, defaultValue=<rgba 0 1 0.8966 1>}.RGB, Alpha=@n24.GlowTransition)) {constB=0.05}
  NamedRerouteDeclaration(None=@n25.Erosion) {name=Erosion, nodeColor=<rgba 0 1 0.1094 1>}
  ParticleSubUVProperties()
  VectorParameter() {parameterName=PaperColor, defaultValue=<rgba 0.6354 0.5654 0.513 1>}
 ]
}
