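# Skydiving contrail for the MistRavenCowl cosmetic: books attached to skeletal mesh bones that open and release pages, paper birds and feathers.
# After User.BookOpenDelay, emitters add red aura and glow with lights, curl-noise feathers and pages, and birds randomly flagged to become ravens.
# Shared lean, dive and book-size parameters steer orientation, motion and scale.
# Front-end modifiers, camera fades and time-of-day scalars adapt the look per context.
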NiagaraSystem /BRCosmetics/Athena/Effects/Particles/Skydiving/NS_Contrail_MistRavenCowl/NS_Contrail_MistRavenCowl
dependencies {
 usedDataInterfaces [
  &NiagaraDataInterfaceCurve
  &NiagaraDataInterfaceVectorCurve
  &NiagaraDataInterfaceColorCurve
  &NiagaraDataInterfaceMeshRendererInfo
  &NiagaraDataInterfaceCamera
  &NiagaraDataInterfacePlatformSet
  &NiagaraDataInterfaceSpriteRendererInfo
  &NiagaraDataInterfaceVector2DCurve
 ]
 usedDynamicInputs [
  &ConvertVectorToPosition
  &Vector2DFromCurve
  &RandomRangeVector2D
 ]
 usedModules [
  &self:SystemSpawnScript.SystemScriptSource.SystemScriptGraph.NiagaraNodeAssignment_1.SetVariables_#1
  &SystemState
  &self:SystemSpawnScript.SystemScriptSource.SystemScriptGraph.NiagaraNodeAssignment_0.SetVariables_#2
 ]
}
emitters {
 Big_aura {
  emitterUpdateScript [
   &EmitterState(
    Life Cycle Mode=::NewEnumerator1
    Loop Behavior=::NewEnumerator1
    Loop Delay=&Subtract_Float(A=$User.BookOpenDelay, B=0.05)
    Loop Duration=0.2
    UseLoopDelay=true
   )
   &SpawnBurst_Instantaneous(Spawn Count=10, Use Spawn Probability=false, UseLoopCountLimit=false)
   &SpawnBurst_Instantaneous(Spawn Count=2, Spawn Group=1, Use Spawn Probability=false, UseLoopCountLimit=false) {module=SpawnBurst_Instantaneous001}
   &SpawnBurst_Instantaneous(Spawn Count=2, Spawn Group=2, Use Spawn Probability=false) {module=SpawnBurst_Instantaneous002}
  ]
  particleSpawnScript [
   &PartitionParticles()
   /SetVariables(Particles.VisibilityTag=$Emitter.SpawnGroup) {isSetParameters=true}
   &InitializeParticle(
    Color=&RandomRangeLinearColor(Minimum=<rgba 0 0 0 1>, Maximum=<rgba 1 1 1 1>)
    Lifetime=&MergeHelper_FloatSwitch(
     0=&RandomRangeFloat(Minimum=0.5, Maximum=1)
     1=&RandomRangeFloat(Minimum=0.5, Maximum=1)
     2=0.25
     3=?
     Index=$Particles.VisibilityTag
     Num=::NewEnumerator1
    )
    Sprite Size Mode=::NewEnumerator3
    Sprite UV Mode=::NewEnumerator3
    Uniform Sprite Size=&MergeHelper_FloatSwitch(
     0=&RandomRangeFloat(Minimum=25, Maximum=40)
     1=&RandomRangeFloat(Minimum=25, Maximum=40)
     2=&RandomRangeFloat(Minimum=70, Maximum=75)
     3=?
     Index=$Particles.VisibilityTag
     Num=::NewEnumerator1
    )
    UsePositionOffset=false
    Write Material Random=true
   )
   &CameraOffset(Camera Offset Amount=&MergeHelper_FloatSwitch(0=20, 1=0, 2=50, 3=?, Index=$Particles.VisibilityTag, Num=::NewEnumerator1))
   &AddVelocity(Velocity=<0 0 1>, Velocity Speed Scale=&RandomRangeFloat(Minimum=10, Maximum=20))
   &DynamicMaterialParameters(
    Index 0 Param 1=&RandomRangeFloat(Minimum=0.4, Maximum=0.6)
    Index 0 Param 2=1
    Index 0 Param 3=1
    Index 0 Param 4=1
    Module.Param0WriteEnabled=true
    Module.Param1WriteEnabled=true
    Module.Param2WriteEnabled=true
    Module.Param3WriteEnabled=true
   ) {enabled=false}
   /SetVariables() {isSetParameters=true}
  ]
  particleUpdateScript [
   &ParticleState()
   &SkeletalMeshLocation(
    Bone / Socket Index=$Particles.PartitionParticles.Partition
    Bone Sampling Mode=::NewEnumerator4
    MaskSpawnGroup=false
    Orientation Sampling=::NewEnumerator1
    Position Sampling=::NewEnumerator1
    Sampled Position Offset=&AddVectorToPosition(Position=<0 0 20>, Vector=$System.FE_Lean_Offset)
    Sampled Position Offset Coordinate Space:ENiagaraSkelMeshTransforms=::NewEnumerator2
    Skeletal Mesh=$System.Skeletal Mesh
   )
   &ScaleSpriteSize(
    Scale Sprite Size Mode=::NewEnumerator0
    Uniform Scale Factor=&MergeHelper_FloatSwitch(
     0=&FloatFromCurve(
      FloatCurve=DataInterface() {propertyValues {Curve {
         defaultValue=0
         keys [
          {arriveTangent=4.1785, interpMode=RCIM_Cubic, leaveTangent=4.1785, tangentMode=RCTM_User, time=0, value=0.2}
          {arriveTangent=1.1179, interpMode=RCIM_Cubic, leaveTangent=1.1179, tangentMode=RCTM_User, time=0.2042, value=0.7}
          {arriveTangent=0.0224, interpMode=RCIM_Cubic, leaveTangent=0.0224, tangentMode=RCTM_User, time=1, value=1}
         ]
        }}}
      Scale Curve=2
     )
     1=0
     2=&FloatFromCurve(
      FloatCurve=DataInterface() {propertyValues {Curve {keys [<at 0 1.5 RCIM_Cubic>,<at 1 0>]}}}
      Scale Curve=&FN_FrontEndFloatModifier(Gameplay=1, Front End=0.6)
     )
     3=?
     Index=$Particles.VisibilityTag
     Num=::NewEnumerator1
    )
   ) {module=ScaleSpriteSize001}
   &SubUVAnimation(
    End Frame Range Override=30
    Random Start Frame=true
    "Sprite Renderer "=DataInterface() {propertyValues {SpriteRenderer=&self:Big_aura_0.NiagaraSpriteRendererProperties_0}}
    Start Frame Range Override=&RandomRangeInt(Minimum=0, Maximum=5)
    SubUV Lookup Index:NiagaraFloat=&FloatFromCurve(FloatCurve=DataInterface() {propertyValues {Curve {keys [{arriveTangent=2.5315, interpMode=RCIM_Cubic, leaveTangent=2.5315, tangentMode=RCTM_User, time=0, value=0}, <at 1 1 RCIM_Cubic>]}}})
    SubUV Lookup Index Scale=0.3
    UseEndFrame=false
    UseStartFrame=false
   )
   &ScaleColor(
    Scale Alpha=&FloatFromCurve(
     FloatCurve=DataInterface() {propertyValues {Curve {keys [<at 0.2 1 RCIM_Cubic>, {arriveTangent=-1.3009, interpMode=RCIM_Cubic, leaveTangent=-1.3009, tangentMode=RCTM_User, time=1, value=0}]}}}
     Scale Curve=2
    )
    Scale RGB=&MakeVectorFromLinearColorRGB(Color=&ColorFromCurve(ColorCurve=DataInterface() {propertyValues {
        AlphaCurve {keys [<at 0.1 2>]}
        BlueCurve {keys [<at 0 0.1029>,<at 0.1339 0.027>,<at 0.4289 0.0117>,<at 0.8362 0.0066>]}
        GreenCurve {keys [<at 0 0>,<at 0.1339 0>,<at 0.4289 0>,<at 0.8362 0>]}
        RedCurve {keys [<at 0 5>,<at 0.1339 1>,<at 0.4289 0.0573>,<at 0.8362 0.0312>]}
       }}))
    ScaleA=true
    ScaleRGB=true
   )
   &SolveForcesAndVelocity(Clamp Velocity=false, Limit Acceleration=false)
   &FN_Light_Attributes(
    Color=$Particles.Color
    Create Color Attribute=true
    Create Exponent Attribute=true
    Create Position Attribute=true
    Create Radius Attribute=true
    Exponent=1
    Position=&AddVectorToPosition(Position=$Particles.Position, Vector=<0 0 5>)
    Radius=30
    Scale Color=&FloatFromCurve(FloatCurve=DataInterface() {propertyValues {Curve {keys [<at 0 0>,<at 0.1 1>]}}}, Scale Curve=500)
   )
   /SetVariables(Particles.Dark_Glow_Size=&Multiply_Vector2(
     A=&Vector2DFromFloat(Value=150)
     B=&Vector2DFromFloat(Value=&FloatFromCurve(
       FloatCurve=DataInterface() {propertyValues {Curve {keys [<at 0 1.5 RCIM_Cubic>,<at 1 0.5>]}}}
       Scale Curve=&FN_FrontEndFloatModifier(Gameplay=1, Front End=0.6)
      ))
    )) {isSetParameters=true}
   /SetVariables(Particles.Dark_Glow=&MakeLinearColorFromVectorAndFloat(
     "Vector (RGB)"=&MakeVectorFromLinearColorRGB(Color=&MultiplyLinearColor(
       Linear Color=<rgba 0.2031 0.0176 0.0486 1>
       Scale Factor=&MakeVector4_FromFloat(Float=&FNTimeOfDayScalar(Night Scale=0.5, Morning Scale=0.5, Evening Scale=0.3))
      ))
     "Float (Alpha)"=&Multiply_Float(
      A=&Multiply_Float(A=&FNTimeOfDayScalar(Night Scale=0.5, Morning Scale=0.5, Evening Scale=1), B=1.2)
      B=&FloatFromCurve(FloatCurve=DataInterface() {propertyValues {Curve {keys [<at 0 0>,<at 0.2 1>,<at 0.7 0.6>,<at 1 0>]}}}, Scale Curve=0.5)
     )
    )) {isSetParameters=true}
   /SetVariables(Particles.FlareColor=&MakeLinearColorFromVectorAndFloat(
     "Vector (RGB)"=&MakeVectorFromLinearColorRGB(Color=&ColorFromCurve(ColorCurve=DataInterface() {propertyValues {
         AlphaCurve {keys [<at 0 1 RCIM_Cubic>]}
         BlueCurve {keys [<at 0.2142 0.3849 RCIM_Cubic>,<at 0.5581 0.0674>,<at 1 0.064 RCIM_Cubic>]}
         GreenCurve {keys [<at 0.2142 0 RCIM_Cubic>,<at 0.5581 0>,<at 1 0 RCIM_Cubic>]}
         RedCurve {keys [<at 0.2142 10 RCIM_Cubic>,<at 0.5581 1>,<at 1 0.3594 RCIM_Cubic>]}
        }}))
     "Float (Alpha)"=&FloatFromCurve(
      FloatCurve=DataInterface() {propertyValues {Curve {keys [<at 0 0>,<at 0.2 1>,<at 0.8 1>,<at 1 0>]}}}
      Scale Curve=&FNTimeOfDayScalar(Night Scale=0.5, Morning Scale=0.5, Evening Scale=0.5)
     )
    )) {isSetParameters=true}
  ]
  rendererClasses [&NiagaraLightRendererProperties, &NiagaraSpriteRendererProperties]
  renderers [
   {rendererClass=&NiagaraLightRendererProperties, rendererIndex=0}
   {rendererClass=&NiagaraSpriteRendererProperties, rendererIndex=1}
   {rendererClass=&NiagaraSpriteRendererProperties, rendererIndex=2}
   {rendererClass=&NiagaraSpriteRendererProperties, rendererIndex=3}
  ]
 }
 Feathers {
  emitterUpdateScript [
   &EmitterState(
    Inactive Response=::NewEnumerator1
    Life Cycle Mode=::NewEnumerator1
    Loop Behavior=::NewEnumerator1
    Loop Delay=&Subtract_Float(A=$User.BookOpenDelay, B=0)
    Loop Duration Mode=::NewEnumerator1
    UseLoopDelay=true
   )
   &SpawnRate(SpawnRate=10, Use Spawn Probability=false)
   &SpawnBurst_Instantaneous(
    Spawn Count=&RandomRangeInt(Minimum=8, Maximum=12)
    Use Spawn Probability=false
    UseLoopCountLimit=false
   )
   &SpawnBurst_Instantaneous(
    Spawn Count=&RandomRangeInt(Minimum=8, Maximum=12)
    Spawn Group=1
    Use Spawn Probability=false
    UseLoopCountLimit=false
   ) {module=SpawnBurst_Instantaneous001}
  ]
  particleSpawnScript [
   /SetVariables(Particles.VisibilityTag=$Emitter.SpawnGroup) {isSetParameters=true}
   &PartitionParticles()
   &InitializeParticle(
    Color=&MergeHelper_ColorSwitch(
     0=<rgba 0.0118 0.0108 0.0273 1>
     1=<rgba 0.1198 0.0038 0.0037 1>
     2=?
     3=?
     Index=&RandomRangeInt(Minimum=0, Maximum=1)
     Num=::NewEnumerator0
    )
    Lifetime=&MergeHelper_FloatSwitch(
     0=&RandomRangeFloat(Minimum=0.8, Maximum=1.5)
     1=&RandomRangeFloat(Minimum=0.4, Maximum=1)
     2=?
     3=?
     Index=$Particles.VisibilityTag
     Num=::NewEnumerator0
    )
    Mesh Scale Mode=::NewEnumerator3
    Sprite Rotation Mode=::NewEnumerator2
    Sprite Size=<10 18>
    Sprite Size Mode=::NewEnumerator4
    UsePositionOffset=true
    Write Material Random=true
   )
   &CameraOffset(Camera Offset Amount=30)
   &SkeletalMeshLocation(
    Bone / Socket Index=$Particles.PartitionParticles.Partition
    Bone Sampling Mode=::NewEnumerator4
    MaskSpawnGroup=false
    Orientation Sampling=::NewEnumerator1
    Position Sampling=::NewEnumerator1
    Sampled Position Offset=&AddVectorToPosition(Position=<0 0 10>, Vector=$System.FE_Lean_Offset)
    Sampled Position Offset Coordinate Space:ENiagaraSkelMeshTransforms=::NewEnumerator2
    Skeletal Mesh=$System.Skeletal Mesh
   )
   &AddVelocity(
    Velocity=&MergeHelper_VectorSwitch(
     0=&Add_Vector(
      A=&RandomRangeVector(Minimum=<-15 -15 0>, Maximum=<15 15 15>)
      B=&Lerp_Vector(
       A=<0 0 0>
       B=&RandomRangeVector(Minimum=<0 0 0>, Maximum=<0 -25 1>)
       Alpha=&FN_FrontEndFloatModifier(Gameplay=0, Front End=$User.ParaGlideLeanAlpha)
      )
     )
     1=&RandomRangeVector(Minimum=<-30 -30 10>, Maximum=<30 30 40>)
     2=?
     3=?
     Index=$Particles.VisibilityTag
     Num=::NewEnumerator0
    )
    Velocity Speed Scale=&RandomRangeFloat(Minimum=3, Maximum=8)
   ) {module=AddVelocity001}
   &SubUVAnimation(
    Evaluation Type=::NewEnumerator0
    "Sprite Renderer "=DataInterface() {propertyValues {SpriteRenderer=&self:Feathers_0.NiagaraSpriteRendererProperties_0}}
    SubUV Animation Mode:ENiagara_SubUVLookupModeV2=::NewEnumerator2
    UseEndFrame=false
    UseStartFrame=false
   ) {module=SubUVAnimation001}
   &DynamicMaterialParameters(
    Index 1 Param 2=&RandomRangeFloat(Minimum=4, Maximum=6)
    Module.Index1Float0Write=true
    Module.Index1Float1Write=true
    Module.Index1Float2Write=true
    Module.Index1Float3Write=true
    Write Parameter Index 0=::NewEnumerator0
    Write Parameter Index 1=::NewEnumerator1
   )
  ]
  particleUpdateScript [
   &ParticleState()
   &SpriteRotationRate(Rotation Rate=&RandomRangeFloat(Minimum=-250, Maximum=250))
   &ScaleSpriteSize(
    Scale Sprite Size Mode=::NewEnumerator0
    Uniform Scale Factor=&FloatFromCurve(
     FloatCurve=DataInterface() {propertyValues {Curve {
        defaultValue=1
        keys [
         {arriveTangent=4.1785, interpMode=RCIM_Cubic, leaveTangent=4.1785, tangentMode=RCTM_User, time=0, value=1}
         {arriveTangent=1.1179, interpMode=RCIM_Cubic, leaveTangent=1.1179, tangentMode=RCTM_User, time=0.2042, value=1.6}
         {arriveTangent=0.0224, interpMode=RCIM_Cubic, leaveTangent=0.0224, tangentMode=RCTM_User, time=1, value=2}
        ]
       }}}
     Scale Curve=&Multiply_Float(
      A=0.2
      B=/Niagara/DynamicInputs/Camera/CameraFade(
       Camera Query:NiagaraDataInterfaceCamera=DataInterface()
       Fade Offset:NiagaraFloat=50
       Fade Distance:NiagaraFloat=100
       Use Owner Position:NiagaraBool=true
      )
     )
    )
   )
   &ScaleColor(
    Scale Alpha=&Multiply_Float(
     A=&RandomRangeFloat(Minimum=0.5, Maximum=1)
     B=&FloatFromCurve(
      FloatCurve=DataInterface() {propertyValues {Curve {keys [
          {arriveTangent=3.7485, interpMode=RCIM_Cubic, leaveTangent=3.7485, tangentMode=RCTM_User, time=0, value=0.35}
          {interpMode=RCIM_Cubic, tangentMode=RCTM_User, time=0.5, value=1}
          {tangentMode=RCTM_User, time=1, value=0}
         ]}}}
      Scale Curve=2
     )
    )
    ScaleA=true
    ScaleRGB=true
   )
   &Drag(
    Drag=&MergeHelper_FloatSwitch(
     0=&RandomRangeFloat(Minimum=2, Maximum=3)
     1=&RandomRangeFloat(Minimum=0.5, Maximum=1)
     2=?
     3=?
     Index=$Particles.VisibilityTag
     Num=::NewEnumerator0
    )
    Use Linear Drag=true
    Use Rotational Drag=true
   )
   &GravityForce(Gravity=&RandomRangeVector(Minimum=<0 0 200>, Maximum=<0 0 600>))
   &CurlNoiseForce(
    Bias Noise=false
    Noise Frequency=55
    Noise Strength=&RandomRangeFloat(Minimum=400, Maximum=800)
    Pan Noise=false
   )
   &SolveForcesAndVelocity(Clamp Velocity=false, Limit Acceleration=false)
   /SetVariables(
    Particles.RavenColor:LinearColor=&Multiply_LinearColorByFloat(
     Linear Color=<rgba 0 0 0 1>
     Float=&FloatFromCurve(FloatCurve=DataInterface() {propertyValues {Curve {keys [<at 0.6 1 RCIM_Cubic>,<at 1 0 RCIM_Cubic>]}}})
    )
    Particles.RavenScale:Vector3f=&VectorFromFloat(Value=&FloatFromCurve(
      FloatCurve=DataInterface() {propertyValues {Curve {keys [<at 0.6 1 RCIM_Cubic>,<at 1 0 RCIM_Cubic>]}}}
      Scale Curve=&RandomRangeFloat(Minimum=1, Maximum=1.6)
     ))
   ) {isSetParameters=true}
  ]
  rendererClasses [&NiagaraSpriteRendererProperties, &NiagaraMeshRendererProperties]
  renderers [
   {rendererClass=&NiagaraSpriteRendererProperties}
   {rendererClass=&NiagaraMeshRendererProperties, rendererIndex=1}
  ]
 }
 LoosePages {
  emitterUpdateScript [
   &EmitterState(
    Life Cycle Mode=::NewEnumerator1
    Loop Behavior=::NewEnumerator1
    Loop Delay=$User.BookOpenDelay
    Loop Duration Mode=::NewEnumerator1
    UseLoopDelay=true
   )
   &SpawnRate(
    SpawnRate=&FN_FrontEndFloatModifier(Gameplay=&Lerp_Float(A=6, B=10, Alpha=$User.ParaGlideLeanAlpha), Front End=6)
    Use Spawn Probability=false
   ) {module=SpawnRate001}
  ]
  particleSpawnScript [
   &InitializeParticle(
    Color=<rgba 1 0.5467 0.15 1>
    Lifetime=&FN_FrontEndFloatModifier(
     Gameplay=&RandomRangeFloat(Minimum=1.1, Maximum=1.5)
     Front End=&RandomRangeFloat(Minimum=0.7, Maximum=1.5)
    )
    Mass Mode=::NewEnumerator1
    Mesh Renderer Array Visibility Mode=::NewEnumerator3
    Mesh Renderer Info=DataInterface() {propertyValues {MeshRenderer=&self:LoosePages_0.NiagaraMeshRendererProperties_0}}
    Mesh Scale=&Multiply_VectorByFloat(Vector=<1.2081 2.7018 1>, Float=&Multiply_Float(A=1.3511, B=$User.BookSize))
    Mesh Scale Mode=::NewEnumerator4
    UsePositionOffset=false
    Write Material Random=true
   )
   &FN_SetSpawnGroup(Number of Spawn Groups=2)
   &AddVelocity(
    Cone Axis=&Lerp_Vector(
     A=<0 0 1>
     B=<0 -0.45 1>
     Alpha=&FN_FrontEndFloatModifier(Gameplay=0, Front End=$User.ParaGlideLeanAlpha)
    )
    Module.UseConeFalloff=false
    Module.UseVelDistribution=false
    Override Seed=false
    Velocity Mode=::NewEnumerator2
    Velocity Speed=&FN_FrontEndFloatModifier(Gameplay=0, Front End=&RandomRangeFloat(Minimum=140, Maximum=180))
   ) {module=AddVelocity002}
   &InitialMeshOrientation(Use Rotation Vector=false)
   &SkeletalMeshLocation(
    Bone / Socket Index=$Particles.SpawnGroup
    Bone Sampling Mode=::NewEnumerator4
    MaskSpawnGroup=false
    Orientation Sampling=::NewEnumerator1
    Position Sampling=::NewEnumerator1
    Sampled Position Offset=&AddVectorToPosition(Position=<0 0 15>, Vector=$System.FE_Lean_Offset)
    Sampled Position Offset Coordinate Space:ENiagaraSkelMeshTransforms=::NewEnumerator2
    Skeletal Mesh=$System.Skeletal Mesh
   )
   &SubUVAnimation(
    Direct Frame Index:NiagaraFloat=&RandomRangeFloat(Minimum=0, Maximum=1)
    Mesh Renderer=DataInterface() {propertyValues {MeshRenderer=&self:LoosePages_0.NiagaraMeshRendererProperties_0}}
    Renderer Type:ENiagaraMeshOrSprite=::NewEnumerator1
    SubUV Animation Mode:ENiagara_SubUVLookupModeV2=::NewEnumerator5
   )
  ]
  particleUpdateScript [
   &ParticleState()
   &UpdateMeshOrientation(Rotation Rate=&RandomRangeFloat(Minimum=-1, Maximum=1), Rotation Vector=<1 1 1>)
   &ScaleMeshSize(Scale Factor=&FN_FrontEndVector3Modifier(
     Gameplay=&Multiply_VectorByFloat(
      Vector=&Lerp_Vector(
       A=&VectorFromFloat(Value=&FloatFromCurve(FloatCurve=DataInterface() {propertyValues {Curve {keys [
             <at 0.15 0 RCIM_Cubic>
             {interpMode=RCIM_Cubic, tangentMode=RCTM_User, time=0.31, value=1}
             {interpMode=RCIM_Cubic, tangentMode=RCTM_User, time=0.7, value=1}
            ]}}}))
       B=&VectorFromFloat(Value=&FloatFromCurve(FloatCurve=DataInterface() {propertyValues {Curve {keys [
             <at 0.1 0 RCIM_Cubic>
             {arriveTangent=5, interpMode=RCIM_Cubic, leaveTangent=5, time=0.17, value=0.58}
             {interpMode=RCIM_Cubic, tangentMode=RCTM_User, time=0.3, value=1}
             {interpMode=RCIM_Cubic, tangentMode=RCTM_User, time=0.7, value=1}
            ]}}}))
       Alpha=$User.ParaGlideLeanAlpha
      )
      Float=&CameraFade(
       Camera Query:NiagaraDataInterfaceCamera=DataInterface() {propertyValues {bRequireCurrentFrameData=false}}
       Fade Offset:NiagaraFloat=150
       Fade Distance:NiagaraFloat=150
       Use Owner Position:NiagaraBool=true
      )
     )
     Front End=&VectorFromFloat(Value=&FloatFromCurve(FloatCurve=DataInterface() {propertyValues {Curve {keys [
           <at 0.1 0 RCIM_Cubic>
           {interpMode=RCIM_Cubic, tangentMode=RCTM_User, time=0.2, value=1}
           {interpMode=RCIM_Cubic, tangentMode=RCTM_User, time=0.7, value=1}
          ]}}}))
    ))
   &InheritSourceMovement(
    Applied Position Scale=&MakeVector(X=-0.038, Y=-0.038, Z=-0.1)
    ApplyForce=false
    ApplyPosition=true
    ApplyPositionFalloff=false
   )
   &CurlNoiseForce(
    Noise Frequency=&Multiply_Float(
     A=&RandomRangeFloat(Minimum=250, Maximum=500)
     B=&Lerp_Float(A=1, B=4, Alpha=$System.Dive)
    )
    Noise Strength=&Multiply_Float(
     A=&RandomRangeFloat(Minimum=800, Maximum=1000)
     B=&FloatFromCurve(FloatCurve=DataInterface() {propertyValues {Curve {keys [{interpMode=RCIM_Cubic, tangentMode=RCTM_User, time=0, value=0}, {interpMode=RCIM_Cubic, tangentMode=RCTM_User, time=0.4, value=1}]}}})
    )
    Pan Noise=false
   )
   &SolveForcesAndVelocity(Clamp Velocity=false, Limit Acceleration=false)
   &DynamicMaterialParameters(
    Index 0 Param 1=&FN_FrontEndFloatModifier(
     Gameplay=&FloatFromCurve(
      FloatCurve=DataInterface() {propertyValues {Curve {keys [
          {interpMode=RCIM_Cubic, tangentMode=RCTM_User, time=0, value=1}
          {arriveTangent=-5.0569, interpMode=RCIM_Cubic, leaveTangent=-5.0569, tangentMode=RCTM_User, time=0.6, value=0}
         ]}}}
      CurveIndex=$Particles.Age
     )
     Front End=&FloatFromCurve(
      FloatCurve=DataInterface() {propertyValues {Curve {keys [
          {interpMode=RCIM_Cubic, tangentMode=RCTM_User, time=0, value=1}
          {arriveTangent=-5.0569, interpMode=RCIM_Cubic, leaveTangent=-5.0569, tangentMode=RCTM_User, time=0.41, value=0}
         ]}}}
      CurveIndex=$Particles.Age
     )
    )
    Index 1 Param 2=&FloatFromCurve(FloatCurve=DataInterface() {propertyValues {Curve {keys [
         {arriveTangent=-0.0116, interpMode=RCIM_Cubic, leaveTangent=-0.0116, tangentMode=RCTM_User, time=0.2, value=0}
         {arriveTangent=0.8333, interpMode=RCIM_Cubic, leaveTangent=0.8333, time=0.5, value=0.3}
         {arriveTangent=1.4, interpMode=RCIM_Cubic, leaveTangent=1.4, time=0.8, value=0.5}
         {interpMode=RCIM_Cubic, leaveTangent=-0.0476, time=1, value=1}
        ]}}})
    Module.Index0Float0Write=true
    Module.Index0Float1Write=true
    Module.Index0Float2Write=true
    Module.Index0Float3Write=true
    Module.Index1Float0Write=true
    Module.Index1Float1Write=true
    Module.Index1Float2Write=true
    Module.Index1Float3Write=true
    Write Parameter Index 1=::NewEnumerator1
   )
  ]
  rendererClasses [&NiagaraMeshRendererProperties]
  renderers [{rendererClass=&NiagaraMeshRendererProperties, rendererIndex=0}]
 }
 NE_Book {
  emitterUpdateScript [
   &EmitterState(Life Cycle Mode=::NewEnumerator1, Loop Behavior=::NewEnumerator1, Loop Duration=1, UseLoopDelay=false)
   &SpawnBurst_Instantaneous(Spawn Count=2, Use Spawn Probability=false, UseLoopCountLimit=false)
   &SpawnBurst_Instantaneous(Spawn Count=2, Spawn Group=1, Spawn Time=$User.BookOpenDelay, Use Spawn Probability=false, UseLoopCountLimit=false) {module=SpawnBurst_Instantaneous001}
  ]
  particleSpawnScript [
   /SetVariables(Particles.VisibilityTag=$Emitter.SpawnGroup) {isSetParameters=true}
   &InitializeParticle(
    Color Mode=::NewEnumerator0
    Lifetime=$User.BookOpenDelay
    Mesh Scale Mode=::NewEnumerator3
    Mesh Uniform Scale=0.65
    UsePositionOffset=false
    Write Material Random=true
   )
   &FN_SetSpawnGroup(Number of Spawn Groups=2)
   /SetVariables(Particles.NewRotation:Quat4f=<0 0 -44.723 1>) {isSetParameters=true}
   /SetVariables() {isSetParameters=true}
  ]
  particleUpdateScript [
   /SetVariables(Particles.Lean:NiagaraFloat=&Lerp_Float(
     A=$Particles.Lean
     B=&DotProduct(
      Vector A=&TransformVector(Vector=<1 0 0>, Destination Space=::World)
      Vector B=&NormalizeVector(Vector To Normalize=$Engine.Owner.Velocity, Safe Normalize=true, Safe Fallback Threshold=0, Safe Fallback Vector=<0 0 0.001>)
     )
     Alpha=0.1
    )) {isSetParameters=true}
   &SkeletalMeshLocation(
    Bone / Socket Index=$Particles.SpawnGroup
    Bone Sampling Mode=::NewEnumerator4
    MaskSpawnGroup=false
    Orientation Sampling=::NewEnumerator1
    Sampled Position Offset=&MergeHelper_VectorSwitch(
     0=&AddVectorToPosition(Position=<0 0 17>, Vector=$System.FE_Lean_Offset)
     1=&AddVectorToPosition(Position=<0 0 13>, Vector=$System.FE_Lean_Offset)
     2=?
     3=?
     Index=$Particles.VisibilityTag
     Num=::NewEnumerator0
    )
    Sampled Position Offset Coordinate Space:ENiagaraSkelMeshTransforms=::NewEnumerator2
    Skeletal Mesh=$System.Skeletal Mesh
   )
   &ParticleState(Kill Particles When Lifetime Has Elapsed=false)
   &UpdateMeshOrientation(
    Rotation Rate=&FloatFromCurve(
     FloatCurve=DataInterface() {propertyValues {Curve {keys [<at 0 0.2 RCIM_Cubic>, {arriveTangent=-0.2222, interpMode=RCIM_Cubic, leaveTangent=-0.2222, time=0.4, value=1}, <at 0.9 0 RCIM_Cubic>]}}}
     Scale Curve=3
    )
    Rotation Vector=&Multiply_VectorByFloat(
     Vector=&MakeVector(
      X=&Waveform(
       Global Input Position=&RandomRangeFloat(Minimum=0, Maximum=6)
       "Waveform [1]"=::NewEnumerator0
       "[1] Frequency"=&RandomRangeFloat(Minimum=0.7, Maximum=1.5)
       "[1] Amplitude Min/Max"=<0.5 1>
      )
      Y=&Waveform(
       Global Input Position=&RandomRangeFloat(Minimum=0, Maximum=6)
       "Waveform [1]"=::NewEnumerator1
       "[1] Frequency"=&RandomRangeFloat(Minimum=0.7, Maximum=1.5)
       "[1] Amplitude Min/Max"=<0.5 1>
      )
      Z=&Waveform(
       Global Input Position=&RandomRangeFloat(Minimum=0, Maximum=6)
       "Waveform [1]"=::NewEnumerator9
       "[1] Frequency"=&RandomRangeFloat(Minimum=0.7, Maximum=1.5)
       "[1] Amplitude Min/Max"=<0.5 1>
      )
     )
     Float=&MakeCustomFloatFromBool(Bool=&RandomBool(), True Float=-1, False Float=1)
    )
   ) {module=UpdateMeshOrientation001}
   &ScaleMeshSize(Scale Factor=&VectorFromCurve(VectorCurve=DataInterface() {propertyValues {
       XCurve {keys [<at 0 0 RCIM_Cubic>,<at 0.2 1 RCIM_Cubic>]}
       YCurve {keys [<at 0 0 RCIM_Cubic>,<at 0.2 1 RCIM_Cubic>]}
       ZCurve {keys [<at 0 0 RCIM_Cubic>,<at 0.2 1 RCIM_Cubic>]}
      }}))
   &DynamicMaterialParameters(
    Index 0 Param 1=&RandomRangeFloat(Minimum=1.1, Maximum=1.3)
    Index 0 Param 2=1.2
    Index 0 Param 3=&FloatFromCurve(
     FloatCurve=DataInterface() {propertyValues {Curve {keys [{arriveTangent=12.5791, interpMode=RCIM_Cubic, leaveTangent=12.5791, tangentMode=RCTM_User, time=0.95, value=0}, <at 1.1 1 RCIM_Cubic>]}}}
     CurveIndex=&Divide_Float(A=$Engine.System.Age, B=$User.BookOpenDelay)
    )
    Module.Index0Float0Write=true
    Module.Index0Float1Write=true
    Module.Index0Float2Write=true
    Module.Index0Float3Write=true
   )
   /SetVariables(Particles.CustomMeshOrientation:Quat4f=&LerpQuaternion(
     Quaternion A:Quat4f=$Particles.MeshOrientation
     Quaternion B:Quat4f=&MultiplyQuaternion(
      Quaternion A=&MakeQuaternion(
       Quaternion From=0
       Axis=<0 1 0>
       Angle=&Multiply_Float(A=$Particles.Lean, B=-0.12)
       From=?
       To=?
       Angle Type=::NewEnumerator1
       Normalize Vectors=false
      )
      Quaternion B=&LerpQuaternion(
       Quaternion A:Quat4f=<0 0 0 1>
       Quaternion B:Quat4f=<0.2 0 0 1>
       Lerp Factor:NiagaraFloat=&FN_FrontEndFloatModifier(Gameplay=0, Front End=$User.ParaGlideLeanAlpha)
       Enable Spherical Linear Interpolation:NiagaraBool=false
      )
     )
     Lerp Factor:NiagaraFloat=&FloatFromCurve(FloatCurve=DataInterface() {propertyValues {Curve {keys [{arriveTangent=13.6969, interpMode=RCIM_Cubic, time=0.9, value=0}, {interpMode=RCIM_Cubic, leaveTangent=-1.0356, time=1, value=1}]}}})
     Enable Spherical Linear Interpolation:NiagaraBool=false
    )) {isSetParameters=true}
   /SetVariables(Particles.PageVariation:Vector4f=&MakeVector4(X=&MaskFloatBySpawnGroup(Masked Float=1.3, Passthrough Float=1.2, Spawn Group Mask=1), Y=1, Z=0, W=0)) {isSetParameters=true}
   /SetVariables(
    Particles.PageOrient:Quat4f=&MultiplyQuaternion(
     Quaternion A=&MakeQuaternion(
      Quaternion From=0
      Axis=<0 1 0>
      Angle=&Multiply_Float(A=$Particles.Lean, B=-0.15)
      From=?
      To=?
      Angle Type=::NewEnumerator1
      Normalize Vectors=false
     )
     Quaternion B=&LerpQuaternion(
      Quaternion A:Quat4f=<0 0 0 1>
      Quaternion B:Quat4f=<0.2 0 0 1>
      Lerp Factor:NiagaraFloat=&FN_FrontEndFloatModifier(Gameplay=0, Front End=$User.ParaGlideLeanAlpha)
      Enable Spherical Linear Interpolation:NiagaraBool=false
     )
    )
    Particles.PageOrientInv:Quat4f=&MultiplyQuaternion(
     Quaternion A=$Particles.NewRotation
     Quaternion B=&MultiplyQuaternion(
      Quaternion A=&MakeQuaternion(
       Quaternion From=0
       Axis=<0 1 0>
       Angle=&Multiply_Float(A=$Particles.Lean, B=0.15)
       From=?
       To=?
       Angle Type=::NewEnumerator1
       Normalize Vectors=false
      )
      Quaternion B=&LerpQuaternion(
       Quaternion A:Quat4f=<0 0 0 1>
       Quaternion B:Quat4f=<-0.2 0 0 1>
       Lerp Factor:NiagaraFloat=&FN_FrontEndFloatModifier(Gameplay=0, Front End=$User.ParaGlideLeanAlpha)
       Enable Spherical Linear Interpolation:NiagaraBool=false
      )
     )
    )
    Particles.TextColor:LinearColor=&ColorFromCurve(
     ColorCurve=DataInterface() {propertyValues {
       AlphaCurve {keys [<at 0 1 RCIM_Cubic>]}
       BlueCurve {keys [<at 0.1796 0 RCIM_Cubic>,<at 0.3917 0>,<at 0.5922 0>,<at 0.6788 0.0835 RCIM_Cubic>,<at 1 0>]}
       GreenCurve {keys [<at 0.1796 0 RCIM_Cubic>,<at 0.3917 0>,<at 0.5922 0>,<at 0.6788 0 RCIM_Cubic>,<at 1 0>]}
       RedCurve {keys [<at 0.1796 0 RCIM_Cubic>,<at 0.3917 1>,<at 0.5922 0>,<at 0.6788 5 RCIM_Cubic>,<at 1 0>]}
      }}
     CurveIndex=&Waveform("Waveform [1]"=::NewEnumerator12, "[1] Frequency"=2, "[1] Amplitude Min/Max"=<0 1>)
    )
   ) {isSetParameters=true}
  ]
  rendererClasses [&NiagaraMeshRendererProperties]
  renderers [
   {rendererClass=&NiagaraMeshRendererProperties, rendererIndex=0}
   {rendererClass=&NiagaraMeshRendererProperties, rendererIndex=1}
   {rendererClass=&NiagaraMeshRendererProperties, rendererIndex=2}
  ]
 }
 NE_BookGlow {
  emitterUpdateScript [
   &EmitterState(
    Life Cycle Mode=::NewEnumerator1
    Loop Behavior=::NewEnumerator1
    Loop Delay=$User.BookOpenDelay
    Loop Duration Mode=::NewEnumerator1
    UseLoopDelay=true
   )
   &SpawnBurst_Instantaneous(Spawn Count=2, Use Spawn Probability=false, UseLoopCountLimit=false)
   &SpawnBurst_Instantaneous(Spawn Count=2, Spawn Group=1, Use Spawn Probability=false, UseLoopCountLimit=false) {module=SpawnBurst_Instantaneous001}
   &SpawnRate(SpawnGroup=2, SpawnRate=5, Use Spawn Probability=false)
  ]
  particleSpawnScript [
   /SetVariables(Particles.VisibilityTag=$Emitter.SpawnGroup) {isSetParameters=true}
   &PartitionParticles()
   &InitializeParticle(
    Color=<rgba 0.8281 0.0259 0.0553 0.774>
    Lifetime=&MergeHelper_FloatSwitch(0=0.5, 1=0.5, 2=1.5, 3=?, Index=$Particles.VisibilityTag, Num=::NewEnumerator1)
    Mesh Scale=&RandomRangeVector(Minimum=<0.4 0.4 0.4>, Maximum=<0.6 0.6 0.6>)
    Mesh Scale Mode=::NewEnumerator4
    Sprite Rotation Angle=0.1156
    Sprite Rotation Mode=::NewEnumerator1
    Sprite Size Mode=::NewEnumerator3
    Uniform Sprite Size=&MergeHelper_FloatSwitch(0=100, 1=120, 2=0, 3=?, Index=&SpawnGroup(), Num=::NewEnumerator1)
    UsePositionOffset=false
    Write Material Random=true
   )
   &InitialMeshOrientation(Orientation Axis=<0 0 1>, Orientation Vector=<0 0 1>, Rotation=<0 0.25 0>, Use Orientation Vector=true, Use Rotation Vector=true)
   &CameraOffset(Camera Offset Amount=&MergeHelper_FloatSwitch(0=5, 1=5, 2=15, 3=?, Index=$Particles.VisibilityTag, Num=::NewEnumerator1))
   /SetVariables() {isSetParameters=true}
   &AddVelocity(
    Velocity=<0 0 1>
    Velocity Speed Scale=&MergeHelper_FloatSwitch(
     0=0
     1=0
     2=&RandomRangeFloat(Minimum=10, Maximum=20)
     3=?
     Index=$Particles.VisibilityTag
     Num=::NewEnumerator1
    )
   )
  ]
  particleUpdateScript [
   &ParticleState(Kill Particles When Lifetime Has Elapsed=false, Let Infinitely Lived Particles Die When Emitter Deactivates=true)
   &SkeletalMeshLocation(
    Bone / Socket Index=$Particles.PartitionParticles.Partition
    Bone Sampling Mode=::NewEnumerator4
    MaskSpawnGroup=false
    Orientation Sampling=::NewEnumerator1
    Position Sampling=::NewEnumerator1
    Sampled Position Offset=&Add_Vector(
     A=&MergeHelper_VectorSwitch(0=<0 0 15>, 1=<0 0 20>, 2=<0 0 15>, 3=?, Index=$Particles.VisibilityTag, Num=::NewEnumerator1)
     B=$System.FE_Lean_Offset
    )
    Sampled Position Offset Coordinate Space:ENiagaraSkelMeshTransforms=::NewEnumerator2
    Skeletal Mesh=$System.Skeletal Mesh
   ) {module=SkeletalMeshLocation002}
   &ScaleColor(
    Scale Alpha=&FloatFromCurve(FloatCurve=DataInterface() {propertyValues {Curve {keys [<at 0 0>,<at 1 1>]}}})
    Scale RGB=&Multiply_VectorByFloat(
     Vector=&VectorFromFloat(Value=&Waveform(
       Waveform Count=::NewEnumerator1
       "Waveform [1]"=::NewEnumerator9
       "[1] Frequency"=3.5
       "[1] Amplitude Min/Max"=<0 0.6>
       "Waveform [2] Blend"=::NewEnumerator2
       "Waveform [2]"=::NewEnumerator9
       "[2] Input Position"=0
       "[2] Phase"=$Emitter.Age
       "[2] Phase Scale"=1
       "[2] Compound Frequency"=2
       "[2] Frequency"=1.4
       "[2] Amplitude Min/Max"=<0.2 1>
       "[2] Scale"=1
       "[2] Bias"=0
       "[2] Exponent"=1
       "[2] Pulse Duration"=0.1
       "[2] Clip Triangle Min/Max"=<0 0>
       "[2] Random Seed"=0
      ))
     Float=&FNTimeOfDayScalar(Night Scale=0.3, Morning Scale=1, Evening Scale=1)
    )
    ScaleA=true
    ScaleRGB=true
   )
   &FN_Light_Attributes(
    Color=$Particles.Color
    Create Color Attribute=true
    Create Exponent Attribute=true
    Create Position Attribute=true
    Create Radius Attribute=true
    Exponent=1
    Position=&AddVectorToPosition(Position=$Particles.Position, Vector=<0 0 10>)
    Radius=30
    Scale Color=&FloatFromCurve(FloatCurve=DataInterface() {propertyValues {Curve {keys [<at 0 0>,<at 0.2 1>]}}}, Scale Curve=50)
   )
   &DynamicMaterialParameters(
    Index 0 Param 1=&MergeHelper_FloatSwitch(
     0=2
     1=0
     2=&FloatFromCurve(FloatCurve=DataInterface() {propertyValues {Curve {keys [{arriveTangent=2.8297, interpMode=RCIM_Cubic, leaveTangent=2.8297, tangentMode=RCTM_User, time=0, value=-0.1}, <at 1 1>]}}})
     3=?
     Index=$Particles.VisibilityTag
     Num=::NewEnumerator1
    )
    Index 0 Param 2=&MergeHelper_FloatSwitch(0=0, 1=0, 2=1, 3=?, Index=$Particles.VisibilityTag, Num=::NewEnumerator1)
    Index 0 Param 3=&MergeHelper_FloatSwitch(0=0, 1=0, 2=1, 3=?, Index=$Particles.VisibilityTag, Num=::NewEnumerator1)
    Index 1 Param 1=&MergeHelper_FloatSwitch(
     0=&MergeHelper_FloatSwitch(0=-2, 1=2, 2=?, 3=?, Index=$Particles.PartitionParticles.Partition, Num=::NewEnumerator0)
     1=0
     2=0
     3=?
     Index=$Particles.VisibilityTag
     Num=::NewEnumerator1
    )
    Module.Index0Param0Write=true
    Module.Index0Param1Write=true
    Module.Index0Param2Write=true
    Module.Index0Param3Write=true
    Module.Index1Param0Write=true
    Module.Index1Param1Write=true
    Module.Index1Param2Write=true
    Module.Index1Param3Write=true
    Write Parameter Index 1=true
   )
   /SetVariables(Particles.DecalSize:Vector2f=&Multiply_Vector2D_ByFloat(
     Float=&MakeFloatFromVector(
      Vector=&VectorFromFloat(Value=&Waveform(
        Waveform Count=::NewEnumerator1
        "Waveform [1]"=::NewEnumerator9
        "[1] Phase Scale"=2
        "[1] Frequency"=3.5
        "[1] Amplitude Min/Max"=<0.8 1.2>
        "Waveform [2] Blend"=::NewEnumerator2
        "Waveform [2]"=::NewEnumerator7
        "[2] Input Position"=0
        "[2] Phase"=$Emitter.Age
        "[2] Phase Scale"=1
        "[2] Compound Frequency"=2
        "[2] Frequency"=1.4
        "[2] Amplitude Min/Max"=<1 1.3>
        "[2] Scale"=1
        "[2] Bias"=0
        "[2] Exponent"=1
        "[2] Pulse Duration"=0.1
        "[2] Clip Triangle Min/Max"=<0 0>
        "[2] Random Seed"=0
       ))
      Channel=::NewEnumerator0
     )
     Vector2D=<60 60>
    )) {isSetParameters=true}
   /SetVariables(Particles.GlowPosition:NiagaraPosition=&AddVectorToPosition(Position=$Particles.Position, Vector=<0 0 1.0151>)) {isSetParameters=true}
   /SetVariables(Particles.SpriteFacing=<0 -1 0>) {isSetParameters=true}
   /SetVariables(Particles.BookGlowColor:LinearColor=&MakeLinearColorFromVectorAndFloat(
     "Vector (RGB)"=&MakeVectorFromLinearColorRGB(Color=<rgba 2 2 2 1>)
     "Float (Alpha)"=&FloatFromCurve(FloatCurve=DataInterface() {propertyValues {Curve {keys [<at 0 0>,<at 1 1>]}}})
    )) {isSetParameters=true}
   /SetVariables(Particles.GlowColor=&MakeLinearColorFromVectorAndFloat(
     "Vector (RGB)"=&MakeVectorFromLinearColorRGB(Color=<rgba 1 0 0.05 1>)
     "Float (Alpha)"=&FloatFromCurve(
      FloatCurve=DataInterface() {propertyValues {Curve {keys [{interpMode=RCIM_Cubic, leaveTangent=3.3849, tangentMode=RCTM_Break, time=0, value=0}, <at 1 1>]}}}
      Scale Curve=&Waveform(
       Global Input Position=&RandomRangeFloat(Minimum=0, Maximum=100)
       "Waveform [1]"=::NewEnumerator9
       "[1] Frequency"=&RandomRangeFloat(Minimum=8, Maximum=12)
       "[1] Amplitude Min/Max"=<0.5 1>
      )
     )
    )) {isSetParameters=true}
   &ScaleMeshSize(Scale Factor=&VectorFromFloat(Value=&FloatFromCurve(FloatCurve=DataInterface() {propertyValues {Curve {keys [<at 0 0.2>,<at 0.15 0.5>,<at 1 1.2>]}}})))
   &KillParticles(Kill Particles=&SetBoolByFloatComparison(
     A=&MergeHelper_FloatSwitch(0=0, 1=0, 2=$Particles.NormalizedAge, 3=?, Index=$Particles.VisibilityTag, Num=::NewEnumerator1)
     B=0.6
     Comparison Type=::NewEnumerator2
    ))
   /SetVariables(Particles.ShockwaveColor:LinearColor=&ColorFromCurve(ColorCurve=DataInterface() {propertyValues {
       AlphaCurve {keys [<at 0 0.5>,<at 0.3576 0.5>,<at 1 0>]}
       BlueCurve {keys [<at 0.0131 0.0398>,<at 0.2917 0.0192>,<at 0.7454 0.0052>]}
       GreenCurve {keys [<at 0.0131 0.0044>,<at 0.2917 0.0152>,<at 0.7454 0.0014>]}
       RedCurve {keys [<at 0.0131 0.9271>,<at 0.2917 0.1823>,<at 0.7454 0.0035>]}
      }})) {isSetParameters=true}
   &SolveForcesAndVelocity(Clamp Velocity=false, Limit Acceleration=false)
  ]
  rendererClasses [
   &NiagaraSpriteRendererProperties
   &NiagaraLightRendererProperties
   &NiagaraMeshRendererProperties
  ]
  renderers [
   {rendererClass=&NiagaraSpriteRendererProperties}
   {rendererClass=&NiagaraLightRendererProperties, rendererIndex=1}
   {rendererClass=&NiagaraSpriteRendererProperties, rendererIndex=2}
   {rendererClass=&NiagaraSpriteRendererProperties, rendererIndex=3}
   {rendererClass=&NiagaraMeshRendererProperties, rendererIndex=4}
  ]
 }
 PaperBirds {
  emitterUpdateScript [
   &EmitterState(
    Life Cycle Mode=::NewEnumerator1
    Loop Behavior=::NewEnumerator1
    Loop Delay=$User.BookOpenDelay
    Loop Duration Mode=::NewEnumerator1
    UseLoopDelay=true
   )
   &SpawnRate(SpawnRate=&RandomRangeFloat(Minimum=5, Maximum=6), Use Spawn Probability=false) {module=SpawnRate001}
  ]
  particleSpawnScript [
   &InitializeParticle(
    Color=<rgba 1 0.5467 0.15 1>
    Lifetime=&FN_FrontEndFloatModifier(
     Gameplay=&RandomRangeFloat(Minimum=1.2, Maximum=1.5)
     Front End=&RandomRangeFloat(Minimum=0.5, Maximum=1)
    )
    Mass Mode=::NewEnumerator1
    Mesh Renderer Array Visibility Mode=::NewEnumerator3
    Mesh Renderer Info=DataInterface() {propertyValues {MeshRenderer=&self:PaperBirds_0.NiagaraMeshRendererProperties_0}}
    Mesh Scale=&Multiply_VectorByFloat(Vector=<1 1 1>, Float=&Multiply_Float(A=3, B=$User.BookSize))
    Mesh Scale Mode=::NewEnumerator4
    UsePositionOffset=false
    Write Material Random=true
   )
   /SetVariables(Particles.TransformToRaven:NiagaraBool=&RandomBool()) {isSetParameters=true}
   &FN_SetSpawnGroup(Number of Spawn Groups=2)
   &AddVelocity(
    Cone Angle=50
    Cone Axis=&Lerp_Vector(
     A=<0 0 1>
     B=<0 -0.45 1>
     Alpha=&FN_FrontEndFloatModifier(Gameplay=0, Front End=$User.ParaGlideLeanAlpha)
    )
    Inner Cone Angle=30
    Module.UseConeFalloff=false
    Module.UseVelDistribution=false
    Override Seed=false
    Velocity Mode=::NewEnumerator2
    Velocity Speed=&FN_FrontEndFloatModifier(
     Gameplay=&Lerp_Float(
      A=&RandomRangeFloat(Minimum=150, Maximum=250)
      B=&RandomRangeFloat(Minimum=250, Maximum=350)
      Alpha=$User.ParaGlideLeanAlpha
     )
     Front End=&RandomRangeFloat(Minimum=140, Maximum=180)
    )
   ) {module=AddVelocity002}
   &SkeletalMeshLocation(
    Bone / Socket Index=$Particles.SpawnGroup
    Bone Sampling Mode=::NewEnumerator4
    MaskSpawnGroup=false
    Orientation Sampling=::NewEnumerator1
    Position Sampling=::NewEnumerator1
    Sampled Position Offset=&AddVectorToPosition(Position=<0 0 15>, Vector=$System.FE_Lean_Offset)
    Sampled Position Offset Coordinate Space:ENiagaraSkelMeshTransforms=::NewEnumerator2
    Skeletal Mesh=$System.Skeletal Mesh
   )
  ]
  particleUpdateScript [
   &ParticleState()
   /SetVariables(Particles.CameraLerp:NiagaraFloat=&CameraFade(
     Camera Query:NiagaraDataInterfaceCamera=DataInterface() {propertyValues {bRequireCurrentFrameData=false}}
     Fade Offset:NiagaraFloat=100
     Fade Distance:NiagaraFloat=100
     Use Owner Position:NiagaraBool=true
    )) {isSetParameters=true}
   &ScaleMeshSize(Scale Factor=&VectorFromFloat(Value=&Multiply_Float(
      A=&MakeCustomFloatFromBool(
       Bool=$Particles.TransformToRaven
       True Float=&FloatFromCurve(FloatCurve=DataInterface() {propertyValues {Curve {keys [<at 0.35 1>,<at 0.5 0>]}}})
       False Float=1
      )
      B=&Multiply_Float(
       A=$Particles.CameraLerp
       B=&FloatFromCurve(FloatCurve=DataInterface() {propertyValues {Curve {keys [<at 0 0>,<at 0.1 1>,<at 0.35 1>]}}})
      )
     )))
   &CurlNoiseForce(
    Noise Frequency=&Multiply_Float(A=&RandomRangeFloat(Minimum=10, Maximum=15), B=1)
    Noise Strength=&Multiply_Float(
     A=&RandomRangeFloat(Minimum=800, Maximum=1200)
     B=&FloatFromCurve(FloatCurve=DataInterface() {propertyValues {Curve {keys [{interpMode=RCIM_Cubic, tangentMode=RCTM_User, time=0, value=0}, {interpMode=RCIM_Cubic, tangentMode=RCTM_User, time=0.4, value=1}]}}})
    )
    Pan Noise=false
   )
   &SolveForcesAndVelocity(Clamp Velocity=false, Limit Acceleration=false)
   &DynamicMaterialParameters(
    Index 0 Param 1=&FN_FrontEndFloatModifier(
     Gameplay=&FloatFromCurve(
      FloatCurve=DataInterface() {propertyValues {Curve {keys [
          {interpMode=RCIM_Cubic, tangentMode=RCTM_User, time=0, value=1}
          {arriveTangent=-5.0569, interpMode=RCIM_Cubic, leaveTangent=-5.0569, tangentMode=RCTM_User, time=0.6, value=0}
         ]}}}
      CurveIndex=$Particles.Age
     )
     Front End=&FloatFromCurve(
      FloatCurve=DataInterface() {propertyValues {Curve {keys [
          {interpMode=RCIM_Cubic, tangentMode=RCTM_User, time=0, value=1}
          {arriveTangent=-5.0569, interpMode=RCIM_Cubic, leaveTangent=-5.0569, tangentMode=RCTM_User, time=0.41, value=0}
         ]}}}
      CurveIndex=$Particles.Age
     )
    )
    Index 0 Param 2=&FloatFromCurve(FloatCurve=DataInterface() {propertyValues {Curve {keys [<at 0 0>,<at 0.2 1>,<at 0.5 0>]}}})
    Index 0 Param 3=&RandomRangeFloat(Minimum=0, Maximum=0.5)
    Index 1 Param 2=&MakeCustomFloatFromBool(
     Bool=$Particles.TransformToRaven
     True Float=&FloatFromCurve(FloatCurve=DataInterface() {propertyValues {Curve {keys [
          {arriveTangent=-0.0116, interpMode=RCIM_Cubic, leaveTangent=-0.0116, tangentMode=RCTM_User, time=0.2, value=0}
          {interpMode=RCIM_Cubic, leaveTangent=1.6667, time=0.3, value=0.3}
         ]}}})
     False Float=&FloatFromCurve(FloatCurve=DataInterface() {propertyValues {Curve {keys [
          {arriveTangent=-0.0116, interpMode=RCIM_Cubic, leaveTangent=-0.0116, tangentMode=RCTM_User, time=0.2, value=0}
          {arriveTangent=0.8333, interpMode=RCIM_Cubic, leaveTangent=0.8333, time=0.5, value=0.3}
          {arriveTangent=1.4, interpMode=RCIM_Cubic, leaveTangent=1.4, time=0.8, value=0.5}
          {interpMode=RCIM_Cubic, leaveTangent=-0.0476, time=1, value=1}
         ]}}})
    )
    Module.Index0Float0Write=true
    Module.Index0Float1Write=true
    Module.Index0Float2Write=true
    Module.Index0Float3Write=true
    Module.Index1Float0Write=true
    Module.Index1Float1Write=true
    Module.Index1Float2Write=true
    Module.Index1Float3Write=true
    Write Parameter Index 1=::NewEnumerator1
   )
   /SetVariables(
    Particles.RavenColor:LinearColor=&ColorFromCurve(ColorCurve=DataInterface() {propertyValues {
       AlphaCurve {keys [<at 0.289 0 RCIM_Cubic>, {arriveTangent=1.6267, interpMode=RCIM_Cubic, leaveTangent=1.6267, time=0.4452, value=1}, <at 0.9037 1>, <at 1 0>]}
       BlueCurve {keys [<at 0 0 RCIM_Cubic>]}
       GreenCurve {keys [<at 0 0 RCIM_Cubic>]}
       RedCurve {keys [<at 0 0 RCIM_Cubic>]}
      }})
    Particles.RavenScale:Vector3f=&Multiply_VectorByFloat(
     Vector=&Multiply_Vector(
      A=$Particles.Initial.Scale
      B=&VectorFromCurve(
       VectorCurve=DataInterface() {propertyValues {
         XCurve {keys [<at 0.9 1>,<at 1 0>,<at 1 0>]}
         YCurve {keys [<at 0.9 1>,<at 1 0>,<at 1 0>]}
         ZCurve {keys [<at 0.4 0.1>,<at 0.6 1>,<at 0.6 1>,<at 0.9 1>,<at 1 0>]}
        }}
       Scale Curve=&VectorFromFloat(Value=&MakeCustomFloatFromBool(Bool=$Particles.TransformToRaven, True Float=&RandomRangeFloat(Minimum=1, Maximum=1.2), False Float=0))
      )
     )
     Float=$Particles.CameraLerp
    )
   ) {isSetParameters=true}
   /SetVariables(
    Particles.DynamicRaven:Vector4f=&MakeVector4(
     X=&FloatFromCurve(FloatCurve=DataInterface() {propertyValues {Curve {keys [<at 0 1>,<at 0.3 0>]}}})
     Y=&RandomRangeFloat(Minimum=3.3, Maximum=4.2)
     Z=&FloatFromCurve(FloatCurve=DataInterface() {propertyValues {Curve {keys [<at 0 0.5>,<at 0.05 5>,<at 0.9 5>,<at 1 0.5>]}}})
     W=0
    )
    Particles.DynamicRaven001:Vector4f=&MakeVector4(X=&RandomRangeFloat(Minimum=0, Maximum=1111), Y=0, Z=0, W=0)
   ) {isSetParameters=true}
  ]
  rendererClasses [&NiagaraMeshRendererProperties]
  renderers [
   {rendererClass=&NiagaraMeshRendererProperties, rendererIndex=0}
   {rendererClass=&NiagaraMeshRendererProperties, rendererIndex=1}
  ]
 }
 Shockwave_Anticipation {
  emitterUpdateScript [
   &EmitterState(Life Cycle Mode=::NewEnumerator1, Loop Behavior=::NewEnumerator1, Loop Duration=0.2, UseLoopDelay=true)
   &SpawnBurst_Instantaneous(Spawn Count=2, Use Spawn Probability=false, UseLoopCountLimit=false)
   &SpawnBurst_Instantaneous(Spawn Count=2, Spawn Time=2.2833, Use Spawn Probability=false, UseLoopCountLimit=false) {module=SpawnBurst_Instantaneous001}
  ]
  particleSpawnScript [
   &PartitionParticles()
   &InitializeParticle(
    Color=<rgba 1 1 1 0.2>
    Lifetime=1.15
    Mesh Scale Mode=::NewEnumerator3
    Mesh Uniform Scale=&RandomRangeFloat(Minimum=0.7, Maximum=0.8)
    UsePositionOffset=false
    Write Material Random=true
   )
   &InitialMeshOrientation(
    Orientation Axis=&RandomRangeVector(Minimum=<0 -3 1>, Maximum=<0 3 1>)
    Orientation Vector=&RandomRangeVector(Minimum=<0 0 1>, Maximum=<0 0 1>)
    Rotation=<0 0.25 0>
    Use Orientation Vector=true
    Use Rotation Vector=true
   )
   &CameraOffset()
   &DynamicMaterialParameters(
    Index 0 Param 1=&FloatFromCurve(FloatCurve=DataInterface() {propertyValues {Curve {keys [{arriveTangent=3.726, interpMode=RCIM_Cubic, leaveTangent=3.726, tangentMode=RCTM_User, time=0, value=-0.3}, <at 1 1>]}}})
    Index 0 Param 2=1
    Index 0 Param 3=1
    Module.Param0WriteEnabled=false
    Module.Param1WriteEnabled=true
    Module.Param2WriteEnabled=true
    Module.Param3WriteEnabled=false
   ) {module=DynamicMaterialParameters001}
  ]
  particleUpdateScript [
   &ParticleState()
   &UpdateMeshOrientation(
    Rotation Rate=&FloatFromCurve(
     FloatCurve=DataInterface() {propertyValues {Curve {keys [{arriveTangent=-5.1999, interpMode=RCIM_Cubic, leaveTangent=-5.1999, tangentMode=RCTM_User, time=0, value=1}, <at 0.4 0 RCIM_Cubic>]}}}
     Scale Curve=&RandomRangeFloat(Minimum=-1, Maximum=1)
    )
    Rotation Vector=<0 0 5>
   )
   &SkeletalMeshLocation(
    Bone / Socket Index=$Particles.PartitionParticles.Partition
    Bone Sampling Mode=::NewEnumerator4
    MaskSpawnGroup=false
    Orientation Sampling=::NewEnumerator1
    Position Sampling=::NewEnumerator1
    Sampled Position Offset=&AddVectorToPosition(Position=<0 0 15>, Vector=$System.FE_Lean_Offset)
    Sampled Position Offset Coordinate Space:ENiagaraSkelMeshTransforms=::NewEnumerator2
    Skeletal Mesh=$System.Skeletal Mesh
   ) {module=SkeletalMeshLocation002}
   &ScaleColor(
    Linear Color Curve=DataInterface() {propertyValues {
      AlphaCurve {keys [<at 0 0>,<at 0.154 0.2>,<at 0.3128 1>,<at 0.5473 1>,<at 0.9987 0>]}
      BlueCurve {keys [<at 0.0148 0>,<at 0.2974 0.0156>,<at 0.4127 0.0345>,<at 0.5475 0.0065>,<at 0.6284 0.0156>]}
      GreenCurve {keys [<at 0.0148 0>,<at 0.2974 0.0076>,<at 0.4127 0.018>,<at 0.5475 0.0044>,<at 0.6284 0.0076>]}
      RedCurve {keys [<at 0.0148 0>,<at 0.2974 0.0076>,<at 0.4127 0.3111>,<at 0.5475 0.0365>,<at 0.6284 0.0076>]}
     }}
    Scale Mode=::NewEnumerator2
   )
   &DynamicMaterialParameters(
    Index 0 Param 1=&FloatFromCurve(FloatCurve=DataInterface() {propertyValues {Curve {keys [
         {arriveTangent=6.6913, interpMode=RCIM_Cubic, time=0, value=-1}
         {arriveTangent=1.4286, interpMode=RCIM_Cubic, leaveTangent=1.4286, time=0.2, value=0.2}
         {arriveTangent=-2.75, interpMode=RCIM_Cubic, leaveTangent=-2.75, time=0.7, value=0}
         {arriveTangent=0.2836, interpMode=RCIM_Cubic, leaveTangent=0.2836, tangentMode=RCTM_User, time=1, value=-2}
        ]}}})
    Index 0 Param 2=&FloatFromCurve(FloatCurve=DataInterface() {propertyValues {Curve {keys [
         {arriveTangent=4.5784, interpMode=RCIM_Cubic, leaveTangent=4.5784, tangentMode=RCTM_User, time=0, value=-1}
         {arriveTangent=1.5259, interpMode=RCIM_Cubic, leaveTangent=1.5259, tangentMode=RCTM_User, time=0.4, value=1.375}
         {arriveTangent=0.2092, interpMode=RCIM_Cubic, leaveTangent=5.4549, tangentMode=RCTM_Break, time=0.5, value=1.5}
         {arriveTangent=-7.9511, interpMode=RCIM_Cubic, leaveTangent=-7.9511, tangentMode=RCTM_User, time=0.7, value=0.2}
        ]}}})
    Index 0 Param 3=&FloatFromCurve(FloatCurve=DataInterface() {propertyValues {Curve {keys [
         {arriveTangent=5, interpMode=RCIM_Cubic, time=0, value=-2}
         {arriveTangent=11.6667, interpMode=RCIM_Cubic, leaveTangent=11.6667, time=0.4, value=1.5}
         {interpMode=RCIM_Cubic, leaveTangent=0.0921, time=0.6, value=5}
        ]}}})
    Module.Index0Float0Write=true
    Module.Index0Float1Write=true
    Module.Index0Float2Write=true
    Module.Index0Float3Write=true
   )
   &DynamicMaterialParameters(
    Index 0 Param 1=&FloatFromCurve(FloatCurve=DataInterface() {propertyValues {Curve {keys [
         {arriveTangent=6.6913, interpMode=RCIM_Cubic, time=0, value=-1}
         {arriveTangent=2.4, interpMode=RCIM_Cubic, leaveTangent=2.4, time=0.2, value=0.3}
         {arriveTangent=-2.875, interpMode=RCIM_Cubic, leaveTangent=-2.875, time=0.5, value=0.2}
         {arriveTangent=0.2836, interpMode=RCIM_Cubic, leaveTangent=0.2836, tangentMode=RCTM_User, time=1, value=-2}
        ]}}})
    Index 0 Param 2=&FloatFromCurve(FloatCurve=DataInterface() {propertyValues {Curve {keys [
         {arriveTangent=1.7659, interpMode=RCIM_Cubic, leaveTangent=1.7659, tangentMode=RCTM_User, time=0, value=0}
         {arriveTangent=5.4549, interpMode=RCIM_Cubic, leaveTangent=5.4549, tangentMode=RCTM_User, time=0.4, value=1.5}
         {arriveTangent=-7.9511, interpMode=RCIM_Cubic, leaveTangent=-7.9511, tangentMode=RCTM_User, time=0.6, value=0.2}
        ]}}})
    Index 0 Param 3=&FloatFromCurve(FloatCurve=DataInterface() {propertyValues {Curve {keys [
         {arriveTangent=5, interpMode=RCIM_Cubic, time=0, value=-2}
         {arriveTangent=10, interpMode=RCIM_Cubic, leaveTangent=10, time=0.5, value=1.5}
         {interpMode=RCIM_Cubic, leaveTangent=10.9916, time=0.6, value=4}
        ]}}})
    Module.Index0Float0Write=false
    Module.Index0Float1Write=true
    Module.Index0Float2Write=false
    Module.Index0Float3Write=true
   ) {enabled=false, module=DynamicMaterialParameters002}
   &ScaleMeshSize(Scale Factor=&VectorFromFloat(Value=&Multiply_Float(
      A=&FloatFromCurve(FloatCurve=DataInterface() {propertyValues {Curve {keys [
           {arriveTangent=2.3799, interpMode=RCIM_Cubic, time=0.3, value=1}
           {arriveTangent=0.6667, leaveTangent=0.6667, time=0.4, value=1.2}
           {arriveTangent=-0.6667, leaveTangent=-0.6667, time=0.6, value=1.2}
           {interpMode=RCIM_Cubic, leaveTangent=-1.75, time=0.7, value=1}
          ]}}})
      B=1
     )))
   &SolveForcesAndVelocity(Clamp Velocity=false, Limit Acceleration=false)
  ]
  rendererClasses [&NiagaraMeshRendererProperties]
  renderers [{rendererClass=&NiagaraMeshRendererProperties, rendererIndex=0}]
 }
 TrailLong {
  emitterUpdateScript [
   &EmitterState(
    Life Cycle Mode=::NewEnumerator1
    Loop Behavior=::NewEnumerator1
    Loop Delay=&Subtract_Float(A=$User.BookOpenDelay, B=0)
    Loop Duration Mode=::NewEnumerator1
    UseLoopDelay=true
   )
   &SpawnBurst_Instantaneous(Spawn Count=2, Use Spawn Probability=false, UseLoopCountLimit=false)
   &SpawnBurst_Instantaneous(Spawn Count=2, Spawn Group=1, Use Spawn Probability=false, UseLoopCountLimit=false) {module=SpawnBurst_Instantaneous001}
   &SpawnRate(SpawnGroup=2, SpawnRate=10, Use Spawn Probability=false)
  ]
  particleSpawnScript [
   /SetVariables(Particles.VisibilityTag=$Emitter.SpawnGroup) {isSetParameters=true}
   &PartitionParticles()
   &InitializeParticle(
    Color=&MergeHelper_ColorSwitch(0=<rgba 0 0 0 1.4>, 1=<rgba 10 0.2956 0 2.5>, 2=?, 3=?, Index=$Particles.VisibilityTag, Num=::NewEnumerator0)
    Lifetime=&MakeCustomFloatFromBool(
     Bool=&SetBoolByFloatComparison(A=1, B=$User.ParaGlideLeanAlpha, Comparison Type=::NewEnumerator4)
     True Float=0.7
     False Float=1.2
    )
    Mesh Scale=&MergeHelper_VectorSwitch(
     0=<1 1 1>
     1=<0.85 0.85 0.5>
     2=&RandomRangeVector(Minimum=<1.3 0.8 0.8>, Maximum=<1.2 3 3>)
     3=?
     Index=$Particles.VisibilityTag
     Num=::NewEnumerator1
    )
    Mesh Scale Mode=::NewEnumerator4
    UsePositionOffset=true
    Write Material Random=true
   )
   &InitialMeshOrientation(
    Mesh Orientation Mode=::NewEnumerator2
    Rotation=&MergeHelper_VectorSwitch(0=<0.25 0 0>, 1=<0 0.25 0>, 2=<0 0.25 0>, 3=?, Index=$Particles.VisibilityTag, Num=::NewEnumerator1)
    Use Rotation Vector=true
   )
   &InitialMeshOrientation(
    Mesh Orientation Mode=::NewEnumerator2
    Rotation=&RandomRangeVector(Minimum=<0 0 0.2>, Maximum=<0 0 -0.2>)
    Use Rotation Vector=true
   ) {module=InitialMeshOrientation001}
   /SetVariables(Particles.Inertia:NiagaraPosition=$Particles.Position) {isSetParameters=true}
   /SetVariables() {isSetParameters=true}
   /SetVariables(Particles.SpiralOrient:Quat4f=&MakeQuaternion(
     Quaternion From=2
     Axis=<1 0 0>
     Angle=0
     X=&RandomRangeFloat(Minimum=0, Maximum=1)
     Y=-0.25
     From=?
     To=?
     Angle Type=::NewEnumerator1
     Normalize Vectors=false
    )) {isSetParameters=true}
  ]
  particleUpdateScript [
   &ParticleState(Kill Particles When Lifetime Has Elapsed=false, Let Infinitely Lived Particles Die When Emitter Deactivates=true)
   &UpdateMeshOrientation(
    Facing Direction=&Lerp_Vector(
     A=<0 1 0>
     B=&RotateVectorByQuaternion(
      Vector To Rotate=&Lerp_Vector(A=<0 1 0>, B=<0 1 0.5>, Alpha=$User.ParaGlideLeanAlpha)
      Quaternion=$Engine.Owner.Rotation
     )
     Alpha=&FN_FrontEndFloatModifier(Gameplay=0, Front End=$User.ParaGlideLeanAlpha)
    )
    Facing Interpolation=false
    Orientation Method=::NewEnumerator1
    Override Quat=false
    TransformFacing=false
    TransformUp=false
   )
   /SetVariables(Particles.RedTrailOrient:Quat4f=&LerpQuaternion(
     Quaternion A:Quat4f=<0 0 0 1>
     Quaternion B:Quat4f=&MultiplyQuaternion(Quaternion A=$Engine.Owner.Rotation, Quaternion B=<0.25 0 0 1>)
     Lerp Factor:NiagaraFloat=&FN_FrontEndFloatModifier(Gameplay=0, Front End=$User.ParaGlideLeanAlpha)
     Enable Spherical Linear Interpolation:NiagaraBool=false
    )) {isSetParameters=true}
   &ScaleMeshSize(Scale Factor=&MergeHelper_VectorSwitch(
     0=&Multiply_VectorByFloat(
      Vector=&VectorFromCurve(
       VectorCurve=DataInterface() {propertyValues {
         XCurve {keys [<at 0 1>,<at 1 1>]}
         YCurve {keys [{arriveTangent=3.2413, interpMode=RCIM_Cubic, leaveTangent=3.2413, tangentMode=RCTM_User, time=0, value=0}, <at 0.5 1 RCIM_Cubic>]}
         ZCurve {keys [{arriveTangent=3.6524, time=0, value=1}, {leaveTangent=-0.0748, time=0.5, value=1}]}
        }}
       Scale Curve=&MakeCustomVectorFromBool(
        Bool=$User.bIsFrontEndPreview
        True Vector=&Lerp_Vector(A=<1 0.6 1>, B=<1 1.3 1>, Alpha=$User.ParaGlideLeanAlpha)
        False Vector=&Lerp_Vector(A=<1 0.9 1>, B=<1 1.5 1>, Alpha=$User.ParaGlideLeanAlpha)
       )
      )
      Float=&FN_FrontEndFloatModifier(Gameplay=1, Front End=1)
     )
     1=&Multiply_VectorByFloat(
      Vector=&VectorFromCurve(VectorCurve=DataInterface() {propertyValues {
         XCurve {keys [<at 0.1 0.2 RCIM_Cubic>,<at 0.4 1 RCIM_Cubic>]}
         YCurve {keys [<at 0.1 0.2 RCIM_Cubic>,<at 0.4 1 RCIM_Cubic>]}
         ZCurve {keys [
           {arriveTangent=5.2899, interpMode=RCIM_Cubic, leaveTangent=5.2899, tangentMode=RCTM_User, time=0, value=0}
           {arriveTangent=0.283, interpMode=RCIM_Cubic, leaveTangent=0.283, tangentMode=RCTM_User, time=0.3, value=1}
          ]}
        }})
      Float=&FN_FrontEndFloatModifier(Gameplay=1, Front End=0.6)
     )
     2=&Multiply_Vector(
      A=&VectorFromCurve(VectorCurve=DataInterface() {propertyValues {XCurve {keys [<at 0 0.1 RCIM_Cubic>,<at 1 1 RCIM_Cubic>]}, YCurve {keys [<at 0 1>,<at 1 1>]}, ZCurve {keys [<at 0 1>,<at 1 1>]}}})
      B=&Lerp_Vector(A=<1 1 1>, B=<1.3 1 1>, Alpha=$User.ParaGlideLeanAlpha)
     )
     3=?
     Index=$Particles.VisibilityTag
     Num=::NewEnumerator1
    ))
   &SkeletalMeshLocation(
    Bone / Socket Index=$Particles.PartitionParticles.Partition
    Bone Sampling Mode=::NewEnumerator4
    MaskSpawnGroup=false
    Orientation Sampling=::NewEnumerator1
    Sampled Position Offset=&Add_Vector(
     A=&MergeHelper_VectorSwitch(0=<0 0 0>, 1=<0 0 10>, 2=?, 3=?, Index=$Particles.VisibilityTag, Num=::NewEnumerator0)
     B=$System.FE_Lean_Offset
    )
    Sampled Position Offset Coordinate Space:ENiagaraSkelMeshTransforms=::NewEnumerator2
    Skeletal Mesh=$System.Skeletal Mesh
   )
   /SetVariables(Particles.Inertia:Vector3f=&Lerp_Vector(
     A=$Particles.Inertia
     B=Linked() {linkedVariable {name=Particles.Position, type=NiagaraPosition}}
     Alpha=&MergeHelper_FloatSwitch(0=0.05, 1=0.2, 2=?, 3=?, Index=$Particles.VisibilityTag, Num=::NewEnumerator0)
    )) {isSetParameters=true}
   /SetVariables(Particles.Color002=&MakeLinearColorFromVectorAndFloat(
     "Vector (RGB)"=&MakeVectorFromLinearColorRGB(Color=&MultiplyLinearColor(
       Linear Color=&MergeHelper_ColorSwitch(
        0=&MakeCustomLinearColorFromBool(Bool=$User.bIsFrontEndPreview, True Color=<rgba 0.0219 0.0346 0.0729 1>, False Color=<rgba 0.1 0.1 0.1 1>)
        1=<rgba 0 0 0 0>
        2=<rgba 0.1667 0.1667 0.1667 0>
        3=?
        Index=$Particles.VisibilityTag
        Num=::NewEnumerator1
       )
       Scale Factor=&MakeVector4_FromLinearColor(LinearColor=&FNTimeOfDayColor(Night Color=<rgba 0.1927 0.3945 1 1>, Morning Color=<rgba 0.6 0.6 0.6 1>, Day Color=<rgba 0.8 0.8 0.8 1>, Evening Color=<rgba 0.6 0.6 0.6 1>))
      ))
     "Float (Alpha)"=&MergeHelper_FloatSwitch(
      0=0.22
      1=1
      2=&FloatFromCurve(FloatCurve=DataInterface() {propertyValues {Curve {keys [
           {interpMode=RCIM_Cubic, tangentMode=RCTM_User, time=0, value=0}
           {arriveTangent=0.0424, interpMode=RCIM_Cubic, leaveTangent=0.0424, tangentMode=RCTM_User, time=0.2, value=0.8}
           {interpMode=RCIM_Cubic, tangentMode=RCTM_User, time=0.4, value=0.8}
           {tangentMode=RCTM_User, time=1, value=0}
          ]}}})
      3=?
      Index=$Particles.VisibilityTag
      Num=::NewEnumerator1
     )
    )) {isSetParameters=true}
   /SetVariables(Particles.Size_Detail:Vector3f=&Multiply_Vector(
     A=&VectorFromCurve(VectorCurve=DataInterface() {propertyValues {
        XCurve {keys [<at 0 1>,<at 1 1>]}
        YCurve {keys [{arriveTangent=3.2413, interpMode=RCIM_Cubic, leaveTangent=3.2413, tangentMode=RCTM_User, time=0, value=0}, <at 0.5 1 RCIM_Cubic>]}
        ZCurve {keys [{arriveTangent=3.6524, time=0, value=1}, {leaveTangent=-0.0748, time=0.5, value=1}]}
       }})
     B=&MakeCustomVectorFromBool(
      Bool=$User.bIsFrontEndPreview
      True Vector=&Lerp_Vector(A=<1 0.7 1>, B=<1 1.2 1>, Alpha=$User.ParaGlideLeanAlpha)
      False Vector=&Lerp_Vector(A=<1 0.8 1>, B=<1 1.5 1>, Alpha=$User.ParaGlideLeanAlpha)
     )
    )) {isSetParameters=true}
   &ScaleColor(
    Scale Alpha=&MergeHelper_FloatSwitch(
     0=&FloatFromCurve(
      FloatCurve=DataInterface() {propertyValues {Curve {keys [<at 0 0>,<at 0.1 1>]}}}
      Scale Curve=&Multiply_Float(
       A=&FN_FrontEndFloatModifier(Gameplay=1, Front End=0.8)
       B=&FN_PickFloatDependOnPlatform(TargetPlatformActive=1, TargetPlatformNoActive=0.8)
      )
     )
     1=&FloatFromCurve(
      FloatCurve=DataInterface() {propertyValues {Curve {keys [<at 0 0>,<at 0.1 1>]}}}
      Scale Curve=&Multiply_Float(
       A=&FN_FrontEndFloatModifier(Gameplay=1, Front End=0.8)
       B=&FN_PickFloatDependOnPlatform(TargetPlatformActive=1, TargetPlatformNoActive=0.8)
      )
     )
     2=&FloatFromCurve(
      FloatCurve=DataInterface() {propertyValues {Curve {keys [
          {interpMode=RCIM_Cubic, tangentMode=RCTM_User, time=0.1, value=0}
          {arriveTangent=0.0424, interpMode=RCIM_Cubic, leaveTangent=0.0424, tangentMode=RCTM_User, time=0.3, value=0.8}
          {interpMode=RCIM_Cubic, tangentMode=RCTM_User, time=0.4, value=0.8}
          {tangentMode=RCTM_User, time=1, value=0}
         ]}}}
      Scale Curve=&Multiply_Float(A=1, B=&FN_FrontEndFloatModifier(Gameplay=1, Front End=0.8))
     )
     3=?
     Index=$Particles.VisibilityTag
     Num=::NewEnumerator1
    )
    Scale RGB=&Multiply_Vector(
     A=&MakeVectorFromLinearColorRGB(Color=&FNTimeOfDayColor(Night Color=<rgba 0.5 0.625 1 1>, Morning Color=<rgba 1 1 1 1>, Day Color=<rgba 1 1 1 1>, Evening Color=<rgba 1 1 1 1>))
     B=&VectorFromFloat(Value=&FN_FrontEndFloatModifier(Gameplay=1, Front End=0.3))
    )
    ScaleA=true
    ScaleRGB=true
   )
   &DynamicMaterialParameters(
    Index 0 Param 1=&RandomRangeFloat(Minimum=0.2, Maximum=1)
    Index 0 Param 2=&RandomRangeFloat(Minimum=0.2, Maximum=1)
    Index 0 Param 3=&FN_FrontEndFloatModifier(Gameplay=0.5, Front End=0)
    Index 0 Param 4=20
    Index 3 Param 1=150
    Index 3 Param 2=0.3
    Index 3 Param 4=&FloatFromCurve(FloatCurve=DataInterface() {propertyValues {Curve {keys [<at 0 1>,<at 1 1>]}}})
    Module.Index0Float0Write=true
    Module.Index0Float1Write=true
    Module.Index0Float2Write=true
    Module.Index0Float3Write=true
    Module.Index2Float0Write=true
    Module.Index2Float1Write=true
    Module.Index2Float2Write=true
    Module.Index2Float3Write=true
    Module.Index3Float0Write=true
    Module.Index3Float1Write=true
    Module.Index3Float2Write=true
    Module.Index3Float3Write=true
    Write Parameter Index 2=::NewEnumerator1
    Write Parameter Index 3=::NewEnumerator1
   )
   /SetVariables(
    Particles.Interpolate Dynamic=&MakeFloat4FromVectorAndFloat(
     "Vector (RGB)"=$Particles.Inertia
     "Float (Alpha)"=&FN_FrontEndFloatModifier(Gameplay=1, Front End=1)
    )
    Particles.RedDynamic:Vector4f=&MakeVector4(
     X=&RandomRangeFloat(Minimum=0.1, Maximum=0.5)
     Y=&RandomRangeFloat(Minimum=0.1, Maximum=1)
     Z=1
     W=1
    )
    Particles.RedTrailColor:LinearColor=&MultiplyLinearColor(
     Linear Color=$Particles.Initial.Color
     Scale Factor=&MakeVector4FromVectorAndFloat(
      "Vector (RGB)"=&MakeVectorFromLinearColorRGB(Color=&ColorFromCurve(ColorCurve=DataInterface() {propertyValues {
          AlphaCurve {keys [<at 0 1>,<at 1 0>]}
          BlueCurve {keys [<at 0.053 10>, {arriveTangent=-9.5036, leaveTangent=-9.5036, time=0.2569, value=1}, <at 1 1>]}
          GreenCurve {keys [<at 0.053 10>, {arriveTangent=-9.5036, leaveTangent=-9.5036, time=0.2569, value=1}, <at 1 1>]}
          RedCurve {keys [<at 0.053 10>, {arriveTangent=-9.5036, leaveTangent=-9.5036, time=0.2569, value=1}, <at 1 1>]}
         }}))
      "Float (Alpha)"=&Multiply_Float(
       A=&FN_FrontEndFloatModifier(Gameplay=1, Front End=0.8)
       B=&FloatFromCurve(FloatCurve=DataInterface() {propertyValues {Curve {keys [<at 0 0>,<at 0.1 1>]}}}, Scale Curve=0.5)
      )
     )
    )
   ) {isSetParameters=true}
   &KillParticles(Kill Particles=&SetBoolByFloatComparison(
     A=&MergeHelper_FloatSwitch(0=0, 1=0, 2=$Particles.NormalizedAge, 3=?, Index=$Particles.VisibilityTag, Num=::NewEnumerator1)
     B=1
     Comparison Type=::NewEnumerator2
    ))
   /SetVariables(
    Particles.SpiralDyn_1:Vector4f=&MakeVector4(
     X=0
     Y=-0.1
     Z=1
     W=&FloatFromCurve(FloatCurve=DataInterface() {propertyValues {Curve {keys [<at 0 0>,<at 1 1>]}}})
    )
    Particles.SpiralDynamic_3:Vector4f=&MakeVector4(X=500, Y=0, Z=0, W=0)
   ) {isSetParameters=true}
   /SetVariables(Particles.SpiralOrient:Quat4f=&MultiplyQuaternion(
     Quaternion A=$Engine.Owner.Rotation
     Quaternion B=&MultiplyQuaternion(
      Quaternion A=&LerpQuaternion(
       Quaternion A:Quat4f=<0 0 0 1>
       Quaternion B:Quat4f=<0.2 0 0 1>
       Lerp Factor:NiagaraFloat=&FN_FrontEndFloatModifier(Gameplay=0, Front End=$User.ParaGlideLeanAlpha)
       Enable Spherical Linear Interpolation:NiagaraBool=false
      )
      Quaternion B=&MakeQuaternion(
       Quaternion From=2
       Axis=<1 0 0>
       Angle=0
       X=&RandomRangeFloat(Minimum=0, Maximum=1)
       Y=-0.25
       From=?
       To=?
       Angle Type=::NewEnumerator1
       Coordinate Space=::Local
       Normalize Vectors=false
      )
     )
    )) {isSetParameters=true}
  ]
  rendererClasses [&NiagaraMeshRendererProperties]
  renderers [
   {rendererClass=&NiagaraMeshRendererProperties, rendererIndex=0}
   {rendererClass=&NiagaraMeshRendererProperties, rendererIndex=1}
   {rendererClass=&NiagaraMeshRendererProperties, rendererIndex=2}
   {rendererClass=&NiagaraMeshRendererProperties, rendererIndex=3}
   {rendererClass=&NiagaraMeshRendererProperties, rendererIndex=4}
  ]
 }
}
summary {
 systemName=NS_Contrail_MistRavenCowl
 emitters [
  {emitterName=NE_Book, rendererClasses [&NiagaraMeshRendererProperties]}
  {
   emitterName=NE_BookGlow
   rendererClasses [
    &NiagaraSpriteRendererProperties
    &NiagaraLightRendererProperties
    &NiagaraMeshRendererProperties
   ]
  }
  {emitterName=LoosePages, rendererClasses [&NiagaraMeshRendererProperties]}
  {emitterName=PaperBirds, rendererClasses [&NiagaraMeshRendererProperties]}
  {emitterName=TrailLong, rendererClasses [&NiagaraMeshRendererProperties]}
  {emitterName=Feathers, rendererClasses [&NiagaraSpriteRendererProperties, &NiagaraMeshRendererProperties]}
  {emitterName=Shockwave_Anticipation, rendererClasses [&NiagaraMeshRendererProperties]}
  {emitterName=Big_aura, rendererClasses [&NiagaraLightRendererProperties, &NiagaraSpriteRendererProperties]}
 ]
 userVariables [
  User.bIsFrontEnd {type=NiagaraBool}
  User.bIsFrontEndPreview {type=NiagaraBool}
  User.BookOpenDelay {type=NiagaraFloat, defaultValue=0.8, description=""}
  User.BookSize {type=NiagaraFloat, defaultValue=0.6, description=""}
  User.GlowFacing {type=Vector3f, description="", defaultValue {struct=Vector3f, value=<0 0 1>}}
  User.ParaGlideLeanAlpha {type=NiagaraFloat, defaultValue=0, description=""}
  User.SkeletalMesh {
   type=NiagaraDataInterfaceSkeletalMesh
   defaultValue {
    struct=NiagaraExt_VariableValue_DataInterface
    value {dataInterface=&self:NiagaraDataInterfaceSkeletalMesh_0}
   }
  }
 ]
}
